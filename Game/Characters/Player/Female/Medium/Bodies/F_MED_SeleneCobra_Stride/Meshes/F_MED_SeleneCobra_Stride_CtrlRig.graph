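# Control Rig for the SeleneCobra Stride female body that procedurally corrects body deformation and poses the skirt, quiver and chain.
# After a shared DeformRig function, spherical pose readers store thigh up, back, side and inward amounts in variables.
# Parent constraints bind skirt bones to thighs and pelvis, then layered transform offsets blend by those amounts and smoothed jump and frontend-idle flags.
# Chain controls follow skirt bones and drive a stretchy Hermite SplineIK.
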ControlRigBlueprint /Game/Characters/Player/Female/Medium/Bodies/F_MED_SeleneCobra_Stride/Meshes/F_MED_SeleneCobra_Stride_CtrlRig
graphs [
 RigVMFunctionLibrary {links [], nodes []}
 RigVMFunction_Sequence_1_ContainedGraph {nodes [
   n1 = RigVMFunctionEntryNode()
   RigVMFunction_Sequence = &RigVMFunction_Sequence(ExecuteContext=@n1.ExecuteContext)
   n2 = &RigVMFunction_Sequence(ExecuteContext=@RigVMFunction_Sequence.B)
   n3 = &RigVMFunction_Sequence(ExecuteContext=@n2.B)
   n4 = &RigVMFunction_Sequence(ExecuteContext=@n3.B)
   n5 = &RigVMFunction_Sequence(ExecuteContext=@n4.B)
   RigVMFunctionReturnNode(A=@RigVMFunction_Sequence.A, B=@n2.A, C=@n3.A, D=@n4.A, E=@n5.A, F=@n5.B)
  ]}
 RigVMModel {nodes [
   RigUnit_BeginExecution = &RigUnit_BeginExecution()
   RigVMFunction_Sequence = RigVMAggregateNode(ExecuteContext=@RigUnit_BeginExecution.ExecutePin)
   FN_DeformRig_FNC = RigVMFunctionReferenceNode(
    DeformRig_ Curve="(deformRig_l_upperarm_up_alpha_72_#1=False,deformRig_r_upperarm_up_alpha_73_#2=False,deformRig_l_upperarm_fr_alpha_74_#3=False,deformRig_r_upperarm_fr_alpha_75_CB51B8 ...(truncated)"
    DeformRig_Bool="(upperarm_1_#4=True,elbow_3_#5=True,wrist_7_#6=True,groin_8_#7=True,glute_17_#8=True,knee_9_40C408324CA261 ...(truncated)"
    DeformRig_Debug=(upperarm_debug_18_#4=False,wrist_debug_20_#6=False,lat_debug_26_#9=False,groinGlute_debug_27_#7=False)
    DeformRig_Value="(upperarm_up_18_#10=(Rotation=(X=0,Y=0,Z=0,W=1),Translation=(X=0,Y=0,Z=8),Scale3D=(X=0.7,Y=1,Z=1)),upperarm_fwd_fr_19_#11=(Rotation=(X=0,Y=0,Z=0,W=1),Translation=(X=0,Y=10,Z=0),Scale3D=(X=0.5,Y=1,Z= ...(truncated)"
    ExecuteContext=@RigVMFunction_Sequence.A
   )
   > n1 = &RigUnit_SphericalPoseReader(
    ActiveRegionScaleFactors=(PositiveWidth=3.5,NegativeWidth=0.5,PositiveHeight=3,NegativeHeight=3)
    ActiveRegionSize=0.1
    Debug=(bDrawDebug=false,DebugScale=25,DebugSegments=20,DebugThickness=0.25)
    DriverAxis=(X=-1,Y=0,Z=0)
    DriverItem=(Type=Bone,Name="thigh_l")
    FalloffRegionScaleFactors=(PositiveWidth=1,NegativeWidth=1,PositiveHeight=1,NegativeHeight=1)
    FalloffSize=0.45
    OptionalParentItem=(Type=Bone)
    RotationOffset=(X=90,Y=0,Z=0)
   )
   > RigVMVariableNode(Value=@n1.OutputParam) {variable=thigh l up}
   > n2 = &RigUnit_SphericalPoseReader(
    ActiveRegionScaleFactors=(PositiveWidth=3.5,NegativeWidth=0.5,PositiveHeight=3,NegativeHeight=3)
    ActiveRegionSize=0.1
    Debug=(bDrawDebug=false,DebugScale=25,DebugSegments=20,DebugThickness=0.25)
    DriverAxis=(X=1,Y=0,Z=0)
    DriverItem=(Type=Bone,Name="thigh_r")
    FalloffRegionScaleFactors=(PositiveWidth=1,NegativeWidth=1,PositiveHeight=1,NegativeHeight=1)
    FalloffSize=0.45
    FlipHeightScaling=false
    FlipWidthScaling=true
    OptionalParentItem=(Type=Bone)
    RotationOffset=(X=-90,Y=0,Z=0)
   )
   > RigVMVariableNode(Value=@n2.OutputParam) {variable=thigh r up}
   > n3 = &RigUnit_SphericalPoseReader(
    ActiveRegionScaleFactors=(PositiveWidth=2.5,NegativeWidth=0.5,PositiveHeight=2,NegativeHeight=2)
    ActiveRegionSize=0.1
    Debug=(bDrawDebug=false,DebugScale=25,DebugSegments=20,DebugThickness=0.25)
    DriverAxis=(X=-1,Y=0,Z=0)
    DriverItem=(Type=Bone,Name="thigh_l")
    FalloffRegionScaleFactors=(PositiveWidth=0,NegativeWidth=0.55,PositiveHeight=0,NegativeHeight=0)
    FalloffSize=0.2
    OptionalParentItem=(Type=Bone,Name="pelvis")
    RotationOffset=(X=-90,Y=0,Z=-65)
   )
   > RigVMVariableNode(Value=@n3.OutputParam) {variable=thigh l bk}
   > n4 = &RigUnit_SphericalPoseReader(
    ActiveRegionScaleFactors=(PositiveWidth=2.5,NegativeWidth=0.5,PositiveHeight=2,NegativeHeight=2)
    ActiveRegionSize=0.1
    Debug=(bDrawDebug=false,DebugScale=25,DebugSegments=20,DebugThickness=0.25)
    DriverAxis=(X=1,Y=0,Z=0)
    DriverItem=(Type=Bone,Name="thigh_r")
    FalloffRegionScaleFactors=(PositiveWidth=0,NegativeWidth=0.55,PositiveHeight=0,NegativeHeight=0)
    FalloffSize=0.2
    FlipHeightScaling=false
    FlipWidthScaling=true
    OptionalParentItem=(Type=Bone,Name="pelvis")
    RotationOffset=(X=90,Y=0,Z=-65)
   )
   > RigVMVariableNode(Value=@n4.OutputParam) {variable=thigh r bk}
   > n5 = &RigUnit_SphericalPoseReader(
    ActiveRegionScaleFactors=(PositiveWidth=3,NegativeWidth=2,PositiveHeight=3,NegativeHeight=3)
    ActiveRegionSize=0.12
    Debug=(bDrawDebug=false,DebugScale=25,DebugSegments=20,DebugThickness=0.25)
    DriverAxis=(X=-1,Y=0,Z=0)
    DriverItem=(Type=Bone,Name="thigh_l")
    FalloffRegionScaleFactors=(PositiveWidth=1,NegativeWidth=1,PositiveHeight=1,NegativeHeight=1)
    FalloffSize=0.45
    OptionalParentItem=(Type=Bone,Name="pelvis")
    RotationOffset=(X=180,Y=0,Z=0)
   )
   > RigVMVariableNode(Value=@n5.OutputParam) {variable=thigh l side}
   > n6 = &RigUnit_SphericalPoseReader(
    ActiveRegionScaleFactors=(PositiveWidth=3.5,NegativeWidth=3,PositiveHeight=3,NegativeHeight=4)
    ActiveRegionSize=0.12
    Debug=(bDrawDebug=false,DebugScale=25,DebugSegments=20,DebugThickness=0.25)
    DriverAxis=(X=1,Y=0,Z=0)
    DriverItem=(Type=Bone,Name="thigh_r")
    FalloffRegionScaleFactors=(PositiveWidth=1,NegativeWidth=1,PositiveHeight=1,NegativeHeight=1)
    FalloffSize=0.45
    FlipHeightScaling=false
    FlipWidthScaling=true
    OptionalParentItem=(Type=Bone,Name="pelvis")
    RotationOffset=(X=0,Y=0,Z=0)
   )
   > RigVMVariableNode(Value=@n6.OutputParam) {variable=thigh r side}
   > n7 = &RigUnit_SphericalPoseReader(
    ActiveRegionScaleFactors=(PositiveWidth=3.5,NegativeWidth=3,PositiveHeight=3,NegativeHeight=4)
    ActiveRegionSize=0.12
    Debug=(bDrawDebug=false,DebugScale=25,DebugSegments=20,DebugThickness=0.25)
    DriverAxis=(X=1,Y=0,Z=0)
    DriverItem=(Type=Bone,Name="thigh_r")
    FalloffRegionScaleFactors=(PositiveWidth=1,NegativeWidth=0.8,PositiveHeight=1,NegativeHeight=1)
    FalloffSize=0.5
    FlipHeightScaling=false
    FlipWidthScaling=true
    OptionalParentItem=(Type=Bone,Name="pelvis")
    RotationOffset=(X=170,Y=5,Z=0)
   )
   > RigVMVariableNode(Value=@n7.OutputParam) {variable=thigh r in}
   > n8 = &RigUnit_SphericalPoseReader(
    ActiveRegionScaleFactors=(PositiveWidth=3.5,NegativeWidth=3.2,PositiveHeight=3,NegativeHeight=3)
    ActiveRegionSize=0.12
    Debug=(bDrawDebug=false,DebugScale=25,DebugSegments=20,DebugThickness=0.25)
    DriverAxis=(X=-1,Y=0,Z=0)
    DriverItem=(Type=Bone,Name="thigh_l")
    FalloffRegionScaleFactors=(PositiveWidth=1,NegativeWidth=0.7,PositiveHeight=1,NegativeHeight=1)
    FalloffSize=0.5
    OptionalParentItem=(Type=Bone,Name="pelvis")
    RotationOffset=(X=-10,Y=5,Z=0)
   )
   > n9 = RigVMVariableNode(Value=@n8.OutputParam) {variable=thigh l in}
   Less = /Script/RigVM.RigVMFunction_MathDoubleLess(A=RigVMVariableNode() {variable=thigh l up}.Value, B=RigVMVariableNode() {variable=thigh r up}.Value) {is=struct RigVMUnitNode, template="Less::Execute(in A,in B,out Result)"}
   RigVMFunction_MathBoolAnd = &RigVMFunction_MathBoolAnd(A=@Less.Result, B=RigVMVariableNode() {variable=Moving2d}.Value)
   RigVMFunction_ControlFlowBranch = &RigVMFunction_ControlFlowBranch(Condition=@RigVMFunction_MathBoolAnd.Result, ExecuteContext=@n9.ExecuteContext)
   RigVMVariableNode(Value=1, ExecuteContext=@RigVMFunction_ControlFlowBranch.False) {variable=Frt_constraint L}
   > RigVMVariableNode(Value=0) {variable=Frt_constraint R}
   RigVMVariableNode(Value=1, ExecuteContext=@RigVMFunction_ControlFlowBranch.True) {variable=Frt_constraint R}
   > RigVMVariableNode(Value=0) {variable=Frt_constraint L}
   &RigUnit_ParentConstraint(
    AdvancedSettings=(InterpolationType=Average,RotationOrderForFilter=XZY)
    Child=(Type=Bone,Name="dyn_skirtFront_mstr")
    Filter=(TranslationFilter=(bX=false,bY=false,bZ=false),RotationFilter=(bX=True,bY=True,bZ=True),ScaleFilter=(bX=True,bY=True,bZ=True))
    Parents=((Item=(Type=Bone,Name="thigh_l"),Weight=1))
    Weight=1
    ExecutePin=@RigVMFunction_Sequence.B
   )
   > &RigUnit_ParentConstraint(
    AdvancedSettings=(InterpolationType=Average,RotationOrderForFilter=XZY)
    Child=(Type=Bone,Name="dyn_skirtFront_mstr_l")
    Filter=(TranslationFilter=(bX=false,bY=false,bZ=false),RotationFilter=(bX=True,bY=True,bZ=True),ScaleFilter=(bX=True,bY=True,bZ=True))
    Parents=((Item=(Type=Bone,Name="thigh_l"),Weight=1))
    Weight=1
   )
   > &RigUnit_ParentConstraint(
    AdvancedSettings=(InterpolationType=Average,RotationOrderForFilter=XZY)
    Child=(Type=Bone,Name="dyn_skirtSide_mstr_l1")
    Filter=(TranslationFilter=(bX=false,bY=false,bZ=false),RotationFilter=(bX=True,bY=True,bZ=True),ScaleFilter=(bX=True,bY=True,bZ=True))
    Parents=((Item=(Type=Bone,Name="thigh_l"),Weight=0.4),(Item=(Type=Bone,Name="pelvis"),Weight=0.6))
    Weight=1
   )
   > &RigUnit_ParentConstraint(
    AdvancedSettings=(InterpolationType=Average,RotationOrderForFilter=XZY)
    Child=(Type=Bone,Name="dyn_skirtSide_mstr_r")
    Filter=(TranslationFilter=(bX=false,bY=false,bZ=false),RotationFilter=(bX=True,bY=True,bZ=True),ScaleFilter=(bX=True,bY=True,bZ=True))
    Parents=((Item=(Type=Bone,Name="thigh_r"),Weight=0.3),(Item=(Type=Bone,Name="pelvis"),Weight=0.7))
    Weight=1
   )
   > &RigUnit_ParentConstraint(
    AdvancedSettings=(InterpolationType=Average,RotationOrderForFilter=XZY)
    Child=(Type=Bone,Name="dyn_skirtFront_02")
    Parents=((Item=(Type=Bone,Name="thigh_l"),Weight=1))
    Weight=1
   )
   > &RigUnit_ParentConstraint(
    AdvancedSettings=(InterpolationType=Average,RotationOrderForFilter=XZY)
    Child=(Type=Bone,Name="dyn_skirtFront_03")
    Parents=((Item=(Type=Bone,Name="thigh_l"),Weight=1))
    Weight=1
   )
   > &RigUnit_ParentConstraint(
    AdvancedSettings=(InterpolationType=Average,RotationOrderForFilter=XZY)
    Child=(Type=Bone,Name="dyn_skirtFront_04")
    Parents=((Item=(Type=Bone,Name="thigh_l"),Weight=1))
    Weight=1
   )
   > &RigUnit_ParentConstraint(
    AdvancedSettings=(InterpolationType=Average,RotationOrderForFilter=XZY)
    Child=(Type=Bone,Name="dyn_skirtFront_02_l")
    Filter=(TranslationFilter=(bX=false,bY=false,bZ=false),RotationFilter=(bX=True,bY=True,bZ=True),ScaleFilter=(bX=True,bY=True,bZ=True))
    Parents=((Item=(Type=Bone,Name="thigh_l"),Weight=1))
    Weight=1
   )
   > &RigUnit_ParentConstraint(
    AdvancedSettings=(InterpolationType=Average,RotationOrderForFilter=XZY)
    Child=(Type=Bone,Name="dyn_skirtFront_03_l")
    Filter=(TranslationFilter=(bX=false,bY=false,bZ=false),RotationFilter=(bX=True,bY=True,bZ=True),ScaleFilter=(bX=True,bY=True,bZ=True))
    Parents=((Item=(Type=Bone,Name="thigh_twist_01_l"),Weight=1))
    Weight=1
   )
   n10 = RigVMVariableNode() {variable=thigh l side}
   &RigUnit_ModifyTransforms(
    ItemToModify=((Item=(Type=Bone,Name="dyn_skirtFront_mstr_l"),Transform=(Rotation=(X=0,Y=0,Z=-0.0436,W=0.999),Translation=(X=0,Y=0,Z=1),Scale3D=(X=1,Y=1,Z=1))))
    WeightMinimum=0
    ExecutePin=@RigVMFunction_Sequence.D
    Weight=@n10.Value
   )
   > &RigUnit_ModifyTransforms(
    ItemToModify="((Item=(Type=Bone,Name=\"dyn_skirtSide_mstr_l1\"),Transform=(Rotation=(X=0,Y=-0.0436,Z=0,W=0.999),Translation=(X=0,Y=0,Z=1),Scale3D=(X=1,Y=1,Z=1))),(Item=(Type=Bone,Name=\"dyn_skirtSide_02_l\"),Transform=(Rotation=(X=0,Y=0,Z=0,W=1),Translation=(X=0,Y=0,Z=0),Sc ...(truncated)"
    WeightMinimum=0
    Weight=@n10.Value
   )
   > &RigUnit_ModifyTransforms(
    ItemToModify="((Item=(Type=Bone,Name=\"dyn_skirtSide_mstr_r\"),Transform=(Rotation=(X=0.0019,Y=-0.0436,Z=0.0436,W=0.9981),Translation=(X=0,Y=0,Z=1),Scale3D=(X=1,Y=1,Z=1))),(Item=(Type=Bone,Name=\"dyn_skirtSide_02_r\"),Transform=(Rotation=(X=0,Y=0,Z=0,W=1),Translation=(X=0,Y ...(truncated)"
    WeightMinimum=0
    Weight=RigVMVariableNode() {variable=thigh r side}.Value
   )
   AlphaInterp = &RigVMFunction_AlphaInterp(InterpSpeedDecreasing=10, InterpSpeedIncreasing=10, Scale=1, bClampResult=true, Value=RigVMVariableNode() {variable=FrontendIdle}.Value)
   Multiply = &RigVMFunction_MathDoubleMul(B=@AlphaInterp.Result, A=RigVMVariableNode() {variable=thigh r in}.Value)
   > &RigUnit_ModifyTransforms(
    ItemToModify=((Item=(Type=Bone,Name="dyn_skirtFront_mstr"),Transform=(Rotation=(X=-0.0038,Y=-0.0435,Z=-0.0871,W=0.9952),Translation=(X=-2,Y=-2,Z=0),Scale3D=(X=1,Y=1,Z=1))))
    WeightMinimum=0
    Weight=@Multiply.Result
   )
   > &RigUnit_ModifyTransforms(
    ItemToModify=((Item=(Type=Bone,Name="dyn_skirtFront_mstr_l"),Transform=(Rotation=(X=0,Y=-0.1305,Z=0,W=0.9914),Translation=(X=-2,Y=-2,Z=1),Scale3D=(X=1,Y=1,Z=1))))
    WeightMinimum=0
    Weight=@Multiply.Result
   )
   > &RigUnit_ModifyTransforms(
    ItemToModify=((Item=(Type=Bone,Name="dyn_skirtSide_mstr_l1"),Transform=(Rotation=(X=-0.0019,Y=-0.0436,Z=-0.0436,W=0.9981),Translation=(X=0,Y=0,Z=0),Scale3D=(X=1,Y=1,Z=1))))
    WeightMinimum=0
    Weight=@Multiply.Result
   )
   > &RigUnit_ModifyTransforms(
    ItemToModify=((Item=(Type=Bone,Name="dyn_skirtBack_mstr_l"),Transform=(Rotation=(X=0,Y=0,Z=-0.0872,W=0.9962),Translation=(X=0,Y=0,Z=0),Scale3D=(X=1,Y=1,Z=1))))
    WeightMinimum=0
    Weight=@Multiply.Result
   )
   n11 = RigVMVariableNode() {variable=thigh l in}
   > &RigUnit_ModifyTransforms(
    ItemToModify=((Item=(Type=Bone,Name="dyn_skirtFront_mstr"),Transform=(Rotation=(X=-0.0038,Y=-0.0435,Z=-0.0871,W=0.9952),Translation=(X=0,Y=0,Z=0),Scale3D=(X=1,Y=1,Z=1))))
    WeightMinimum=0
    Weight=@n11.Value
   )
   > &RigUnit_ModifyTransforms(
    ItemToModify=((Item=(Type=Bone,Name="dyn_skirtFront_mstr_l"),Transform=(Rotation=(X=-0.0019,Y=-0.0436,Z=-0.0436,W=0.9981),Translation=(X=0,Y=0,Z=0),Scale3D=(X=1,Y=1,Z=1))))
    WeightMinimum=0
    Weight=@n11.Value
   )
   n12 = &RigVMFunction_AlphaInterp(
    InterpSpeedDecreasing=10
    InterpSpeedIncreasing=2
    Scale=1
    bClampResult=true
    bMapRange=true
    Value=RigVMVariableNode() {variable=jumping}.Value
   )
   &RigUnit_ModifyTransforms(
    ItemToModify="((Item=(Type=Bone,Name=\"dyn_skirtFront_04\"),Transform=(Rotation=(X=-0.0864,Y=-0.13,Z=-0.0114,W=0.9877),Translation=(X=0,Y=0,Z=0),Scale3D=(X=1,Y=1,Z=1))),(Item=(Type=Bone,Name=\"dyn_skirtFront_04\"),Transform=(Rotation=(X=0,Y=-0.1305,Z=0,W=0.9914),Translation ...(truncated)"
    WeightMinimum=0
    Weight=@n12.Result
    ExecutePin=@RigVMFunction_Sequence.E
   )
   > &RigUnit_ModifyTransforms(
    ItemToModify="((Item=(Type=Bone,Name=\"dyn_skirtFront_02_l\"),Transform=(Rotation=(X=0,Y=-0.0872,Z=0,W=0.9962),Translation=(X=0,Y=0,Z=0),Scale3D=(X=1,Y=1,Z=1))),(Item=(Type=Bone,Name=\"dyn_skirtFront_03_l\"),Transform=(Rotation=(X=0,Y=-0.0872,Z=0,W=0.9962),Translation=(X=0, ...(truncated)"
    WeightMinimum=0
    Weight=@n12.Result
   )
   n13 = &RigVMFunction_AlphaInterp(
    InterpSpeedDecreasing=10
    InterpSpeedIncreasing=10
    Scale=1
    bClampResult=true
    bMapRange=true
    Value=RigVMVariableNode() {variable=jumping_chain}.Value
   )
   > &RigUnit_ModifyTransforms(
    ItemToModify=((Item=(Type=Bone,Name="dyn_Quiver_MSTR"),Transform=(Rotation=(X=-0.0868,Y=-0.0868,Z=-0.0076,W=0.9924),Translation=(X=0,Y=0,Z=2),Scale3D=(X=1,Y=1,Z=1))))
    WeightMinimum=0
    Weight=@n13.Result
   )
   RigVMFunction_MathQuaternionFromEuler = &RigVMFunction_MathQuaternionFromEuler(Euler=(X=0,Y=0,Z=-10), RotationOrder=XYZ)
   n14 = RigVMVariableNode() {variable=thigh l up}
   &RigUnit_ModifyTransforms(
    ItemToModify="((Item=(Type=Bone,Name=\"dyn_skirtFront_mstr\"),Transform=(Rotation=(X=-0.1371,Y=-0.0747,Z=-0.0974,W=0.9829),Translation=(X=-2,Y=0,Z=0.5),Scale3D=(X=1,Y=1,Z=1))),(Item=(Type=Bone,Name=\"dyn_skirtFront_02\"),Transform=(Rotation=(X=0.1349,Y=0.0259,Z=-0.1349,W=0. ...(truncated)"
    WeightMinimum=0
    ItemToModify.0.Transform.Rotation=@RigVMFunction_MathQuaternionFromEuler.Result
    ItemToModify.1.Transform.Rotation=&RigVMFunction_MathQuaternionFromEuler(Euler=(X=0,Y=0,Z=-10), RotationOrder=XYZ).Result
    ItemToModify.2.Transform.Rotation=&RigVMFunction_MathQuaternionFromEuler(Euler=(X=0,Y=0,Z=-10), RotationOrder=XYZ).Result
    ItemToModify.3.Transform.Rotation=&RigVMFunction_MathQuaternionFromEuler(Euler=(X=15,Y=45,Z=0), RotationOrder=XYZ).Result
    Weight=@n14.Value
    ExecutePin=@RigVMFunction_Sequence.C
   )
   > &RigUnit_ModifyTransforms(
    ItemToModify="((Item=(Type=Bone,Name=\"dyn_skirtFront_mstr_l\"),Transform=(Rotation=(X=0.2088,Y=-0.0704,Z=-0.118,W=0.9683),Translation=(X=0,Y=0,Z=1),Scale3D=(X=1,Y=1,Z=1))),(Item=(Type=Bone,Name=\"dyn_skirtFront_02_l\"),Transform=(Rotation=(X=0,Y=0,Z=-0.0436,W=0.999),Transl ...(truncated)"
    WeightMinimum=0
    ItemToModify.0.Transform.Rotation=&RigVMFunction_MathQuaternionFromEuler(Euler=(X=0,Y=-5,Z=0), RotationOrder=XYZ).Result
    ItemToModify.1.Transform.Rotation=&RigVMFunction_MathQuaternionFromEuler(Euler=(X=-15,Y=0,Z=-25), RotationOrder=XYZ).Result
    ItemToModify.2.Transform.Rotation=&RigVMFunction_MathQuaternionFromEuler(Euler=(X=0,Y=35,Z=-35), RotationOrder=XYZ).Result
    ItemToModify.3.Transform.Rotation=&RigVMFunction_MathQuaternionFromEuler(Euler=(X=0,Y=0,Z=-15), RotationOrder=XYZ).Result
    Weight=@n14.Value
   )
   > &RigUnit_ModifyTransforms(
    ItemToModify="((Item=(Type=Bone,Name=\"dyn_skirtSide_mstr_l1\"),Transform=(Rotation=(X=-0.0262,Y=-0.0831,Z=-0.2996,W=0.9501),Translation=(X=-2,Y=0,Z=3),Scale3D=(X=1,Y=1,Z=1))),(Item=(Type=Bone,Name=\"dyn_skirtSide_02_l\"),Transform=(Rotation=(X=0,Y=0,Z=0,W=1),Translation=(X ...(truncated)"
    WeightMinimum=0
    Weight=@n14.Value
   )
   n15 = RigVMVariableNode() {variable=thigh r up}
   > &RigUnit_ModifyTransforms(
    ItemToModify="((Item=(Type=Bone,Name=\"dyn_skirtSide_mstr_r\"),Transform=(Rotation=(X=0.0019,Y=-0.0436,Z=0.0436,W=0.9981),Translation=(X=0,Y=-3,Z=0),Scale3D=(X=1,Y=1,Z=1))),(Item=(Type=Bone,Name=\"dyn_skirtSide_02_r\"),Transform=(Rotation=(X=0,Y=0,Z=0,W=1),Translation=(X=0, ...(truncated)"
    WeightMinimum=0
    Weight=@n15.Value
   )
   > &RigUnit_ModifyTransforms(
    ItemToModify="((Item=(Type=Bone,Name=\"dyn_skirtBack_mstr_r\"),Transform=(Rotation=(X=0,Y=0,Z=0.0872,W=0.9962),Translation=(X=0,Y=2,Z=0),Scale3D=(X=1,Y=1,Z=1))),(Item=(Type=Bone,Name=\"dyn_skirtBack_02_r\"),Transform=(Rotation=(X=0,Y=-0.0872,Z=0,W=0.9962),Translation=(X=0,Y ...(truncated)"
    WeightMinimum=0
    Weight=@n15.Value
   )
   > &RigUnit_ModifyTransforms(
    ItemToModify=((Item=(Type=Bone,Name="dyn_Quiver_MSTR"),Transform=(Rotation=(X=-0.0261,Y=-0.0871,Z=-0.0023,W=0.9959),Translation=(X=0,Y=0,Z=0),Scale3D=(X=1,Y=1,Z=1))))
    WeightMinimum=0
    Weight=@n15.Value
   )
   > &RigUnit_ModifyTransforms(
    ItemToModify=((Item=(Type=Bone,Name="dyn_skirtFront_01_r"),Transform=(Rotation=(X=-0.0038,Y=-0.0435,Z=-0.0871,W=0.9952),Translation=(X=0,Y=-1,Z=0),Scale3D=(X=1,Y=1,Z=1))))
    WeightMinimum=0
    Weight=@n15.Value
   )
   n16 = &RigVMFunction_Sequence(ExecuteContext=@RigVMFunction_Sequence.F)
   &RigUnit_ParentConstraint(
    AdvancedSettings=(InterpolationType=Average,RotationOrderForFilter=XZY)
    Child=(Type=Bone,Name="dyn_chain_frt_ctrl")
    Filter=(TranslationFilter=(bX=false,bY=false,bZ=false),RotationFilter=(bX=True,bY=True,bZ=True),ScaleFilter=(bX=True,bY=True,bZ=True))
    Parents=((Item=(Type=Bone,Name="dyn_skirtFront_01"),Weight=1),(Item=(Type=Bone,Name="dyn_skirtFront_01_l"),Weight=1))
    Weight=1
    ExecutePin=@n16.A
   )
   > &RigUnit_ParentConstraint(
    AdvancedSettings=(InterpolationType=Average,RotationOrderForFilter=XZY)
    Child=(Type=Bone,Name="dyn_chain_frt_md_MSTR")
    Parents=((Item=(Type=Bone,Name="dyn_skirtFront_02_l"),Weight=1))
    Weight=1
   )
   > &RigUnit_ParentConstraint(
    AdvancedSettings=(InterpolationType=Average,RotationOrderForFilter=XZY)
    Child=(Type=Bone,Name="dyn_chain_mid_MSTR")
    Parents=((Item=(Type=Bone,Name="dyn_skirtSide_02_l"),Weight=1))
    Weight=1
   )
   > &RigUnit_ParentConstraint(
    AdvancedSettings=(InterpolationType=Average,RotationOrderForFilter=XZY)
    Child=(Type=Bone,Name="dyn_chain_bk_md_MSTR")
    Parents=((Item=(Type=Bone,Name="dyn_skirtBack_02_l"),Weight=1))
    Weight=1
   )
   n17 = RigVMVariableNode() {variable=thigh l up}
   > &RigUnit_ModifyTransforms(
    ItemToModify=((Item=(Type=Bone,Name="dyn_chain_frt_ctrl"),Transform=(Rotation=(X=-0.1274,Y=-0.2146,Z=-0.0283,W=0.9679),Translation=(X=1,Y=3,Z=0),Scale3D=(X=1,Y=1,Z=1))))
    WeightMinimum=0
    Weight=@n17.Value
   )
   > &RigUnit_ModifyTransforms(
    ItemToModify=((Item=(Type=Bone,Name="dyn_chain_frt_md_ctrl"),Transform=(Rotation=(X=0,Y=0,Z=0.1305,W=0.9914),Translation=(X=-1,Y=0,Z=3),Scale3D=(X=1,Y=1,Z=1))))
    WeightMinimum=0
    Weight=@n17.Value
   )
   > &RigUnit_ModifyTransforms(
    ItemToModify=((Item=(Type=Bone,Name="dyn_chain_mid_ctrl"),Transform=(Rotation=(X=0.0868,Y=0.0076,Z=0.0868,W=0.9924),Translation=(X=2,Y=0,Z=2),Scale3D=(X=1,Y=1,Z=1))))
    WeightMinimum=0
    Weight=@n17.Value
   )
   > &RigUnit_ModifyTransforms(
    ItemToModify=((Item=(Type=Bone,Name="dyn_chain_bk_md_ctrl"),Transform=(Rotation=(X=0,Y=0,Z=0.1305,W=0.9914),Translation=(X=0,Y=0.5,Z=2),Scale3D=(X=1,Y=1,Z=1))))
    WeightMinimum=0
    Weight=@n17.Value
   )
   > &RigUnit_ModifyTransforms(
    ItemToModify="((Item=(Type=Bone,Name=\"dyn_chain_frt_ctrl\"),Transform=(Rotation=(X=0,Y=0,Z=-0.1305,W=0.9914),Translation=(X=0,Y=1.5,Z=0),Scale3D=(X=1,Y=1,Z=1))),(Item=(Type=Bone,Name=\"dyn_chain_frt_md_ctrl\"),Transform=(Rotation=(X=0,Y=0,Z=0,W=1),Translation=(X=-3,Y=-1,Z= ...(truncated)"
    WeightMinimum=0
    Weight=RigVMVariableNode() {variable=thigh l in}.Value
   )
   n18 = &RigVMFunction_AlphaInterp(
    InterpSpeedDecreasing=10
    InterpSpeedIncreasing=5
    Scale=1
    bClampResult=true
    bMapRange=true
    Value=RigVMVariableNode() {variable=jumping_chain}.Value
   )
   > &RigUnit_ModifyTransforms(
    ItemToModify=((Item=(Type=Bone,Name="dyn_chain_frt_ctrl"),Transform=(Rotation=(X=-0.0831,Y=-0.2996,Z=-0.0262,W=0.9501),Translation=(X=0,Y=0,Z=0),Scale3D=(X=1,Y=1,Z=1))))
    WeightMinimum=0
    Weight=@n18.Result
   )
   > &RigUnit_ModifyTransforms(
    ItemToModify=((Item=(Type=Bone,Name="dyn_chain_frt_md_ctrl"),Transform=(Rotation=(X=0,Y=0,Z=0,W=1),Translation=(X=0,Y=3,Z=3),Scale3D=(X=1,Y=1,Z=1))))
    WeightMinimum=0
    Weight=@n18.Result
   )
   > &RigUnit_ModifyTransforms(
    ItemToModify=((Item=(Type=Bone,Name="dyn_chain_mid_ctrl"),Transform=(Rotation=(X=0,Y=0,Z=0,W=1),Translation=(X=0,Y=3,Z=3),Scale3D=(X=1,Y=1,Z=1))))
    WeightMinimum=0
    Weight=@n18.Result
   )
   > &RigUnit_ModifyTransforms(
    ItemToModify=((Item=(Type=Bone,Name="dyn_chain_bk_md_ctrl"),Transform=(Rotation=(X=0,Y=0,Z=-0.1305,W=0.9914),Translation=(X=0,Y=2,Z=2),Scale3D=(X=1,Y=1,Z=1))))
    WeightMinimum=0
    Weight=@n18.Result
   )
   > &RigUnit_ParentConstraint(
    AdvancedSettings=(InterpolationType=Average,RotationOrderForFilter=XZY)
    Child=(Type=Control,Name="dyn_chain_01_ctrl")
    Parents=((Item=(Type=Bone,Name="dyn_chain_frt_ctrl"),Weight=1))
    Weight=1
   )
   > &RigUnit_ParentConstraint(
    AdvancedSettings=(InterpolationType=Average,RotationOrderForFilter=XZY)
    Child=(Type=Control,Name="dyn_chain_04_ctrl")
    Parents=((Item=(Type=Bone,Name="dyn_chain_frt_md_ctrl"),Weight=1))
    Weight=1
   )
   > &RigUnit_ParentConstraint(
    AdvancedSettings=(InterpolationType=Average,RotationOrderForFilter=XZY)
    Child=(Type=Control,Name="dyn_chain_07_ctrl")
    Parents=((Item=(Type=Bone,Name="dyn_chain_mid_ctrl"),Weight=1))
    Weight=1
   )
   > &RigUnit_ParentConstraint(
    AdvancedSettings=(InterpolationType=Average,RotationOrderForFilter=XZY)
    Child=(Type=Control,Name="dyn_chain_12_ctrl")
    Parents=((Item=(Type=Bone,Name="dyn_chain_bk_md_ctrl"),Weight=1))
    Weight=1
   )
   > &RigUnit_ParentConstraint(
    AdvancedSettings=(InterpolationType=Average,RotationOrderForFilter=XZY)
    Child=(Type=Control,Name="dyn_chain_16_ctrl")
    Parents=((Item=(Type=Bone,Name="dyn_chain_bk_ctrl"),Weight=1))
    Weight=1
   )
   SplineIK = RigVMFunctionReferenceNode(
    Bones="((Type=Bone,Name=\"dyn_chain_01\"),(Type=Bone,Name=\"dyn_chain_02\"),(Type=Bone,Name=\"dyn_chain_03\"),(Type=Bone,Name=\"dyn_chain_04\"),(Type=Bone,Name=\"dyn_chain_05\"),(Type=Bone,Name=\"dyn_chain_06\"),(Type=Bone,Name=\"dyn_chain_07\"),(Type=Bone,Name=\"dyn_chain_08\") ...(truncated)"
    Controls=((Type=Control,Name="dyn_chain_01_ctrl"),(Type=Control,Name="dyn_chain_04_ctrl"),(Type=Control,Name="dyn_chain_07_ctrl"),(Type=Control,Name="dyn_chain_12_ctrl"),(Type=Control,Name="dyn_chain_16_ctrl"))
    Primary Axis=(X=1,Y=0,Z=0)
    Samples Per Segment=16
    Secondary Spline Direction=(X=0,Y=5,Z=0)
    Spline Mode=Hermite
    Stretch=true
    Up Axis=(X=0,Y=1,Z=0)
    debug=true
    ExecuteContext=@n16.B
   )
  ]}
]
hierarchy [
 root {loc=<0 0 0>, rot=<0 0 0 1>}
 pelvis {parent=root, loc=<0 -0.5301 98.0571>, rot=<0.0127 0.707 -0.0127 -0.707>}
 spine_01 {parent=pelvis, loc=<6.1208 -0.9328 0>, rot=<0 0 0.034 -0.9994>}
 spine_02 {parent=spine_01, loc=<6.5996 -0.2095 0>, rot=<0 0 0.0238 -0.9997>}
 spine_03 {parent=spine_02, loc=<6.4914 -0.1237 0>, rot=<0 0 0.0149 -0.9999>}
 spine_04 {parent=spine_03, loc=<6.9598 0.0794 0>, rot=<0 0 -0.0132 -0.9999>}
 spine_05 {parent=spine_04, loc=<6.9987 0 0>, rot=<0 0 -0.0015 -1>}
 attach_cape {parent=spine_05, loc=<9.0262 -11.7144 -0.0007>, rot=<0.0616 -0.7044 -0.0616 -0.7044>}
 neck_01 {parent=spine_05, loc=<11.3138 0.1529 0>, rot=<0 0 -0.2003 -0.9797>}
 neck_02 {parent=neck_01, loc=<4.5627 0 0>, rot=<0 0 0.0504 -0.9987>}
 head {parent=neck_02, loc=<5.3914 0 0>, rot=<0 0 0.115 -0.9934>}
 attach_backpack {parent=spine_05, loc=<-0.1124 -13.4735 -0.0008>, rot=<-0.0093 -0.707 0.0092 -0.707>}
 clavicle_l {parent=spine_05, loc=<5.4831 -0.1053 -3.2946>, rot=<0.0465 -0.7289 -0.009 -0.6829>}
 upperarm_l {parent=clavicle_l, loc=<9.8658 -1.2136 -0.4105>, rot=<0.0164 -0.3245 -0.0269 -0.9454>}
 lowerarm_l {parent=upperarm_l, loc=<26.465 -0.8345 0.259>, rot=<0.0036 0.0161 -0.2298 -0.9731>}
 hand_l {parent=lowerarm_l, loc=<22.7355 -0.3203 -0.6105>, rot=<-0.6784 -0.0223 0.0735 -0.7307>}
 thumb_01_l {parent=hand_l, loc=<1.3118 -0.3191 -2.459>, rot=<0.6546 -0.4042 0.0664 -0.6354>}
 thumb_02_l {parent=thumb_01_l, loc=<5.9336 -0.0448 0.0401>, rot=<-0.0037 0.0213 0.298 -0.9543>}
 thumb_03_l {parent=thumb_02_l, loc=<2.2223 0.3682 -0.0907>, rot=<0.0118 -0.013 0.0066 -0.9998>}
 index_metacarpal_l {parent=hand_l, loc=<1.6535 0.0183 -2.96>, rot=<0.0158 -0.0124 -0.0173 -0.9996>}
 index_01_l {parent=index_metacarpal_l, loc=<8.3008 0.0357 -0.0524>, rot=<-0.0044 0.0485 0.0924 -0.9945>}
 index_02_l {parent=index_01_l, loc=<3.3587 -0.3004 0.0452>, rot=<-0.002 0.0027 0.3225 -0.9466>}
 index_03_l {parent=index_02_l, loc=<2.5578 -0.123 -0.0252>, rot=<-0.002 0.0067 0.2558 -0.9667>}
 middle_metacarpal_l {parent=hand_l, loc=<2.1856 -0.1725 -0.5984>, rot=<0.006 0.0168 -0.0329 -0.9993>}
 middle_01_l {parent=middle_metacarpal_l, loc=<7.3275 -0.2447 0.0129>, rot=<-0.0208 0.0042 0.0858 -0.9961>}
 middle_02_l {parent=middle_01_l, loc=<4.0123 0.1401 0.1175>, rot=<-0.0218 0.0258 0.3512 -0.9357>}
 middle_03_l {parent=middle_02_l, loc=<2.6077 0.0534 0.1373>, rot=<0.0082 0.0188 0.1708 -0.9851>}
 ring_metacarpal_l {parent=hand_l, loc=<2.2074 -0.3625 1.3638>, rot=<-0.0284 0.0722 -0.0458 -0.9959>}
 ring_01_l {parent=ring_metacarpal_l, loc=<7.0295 -0.6197 -0.1446>, rot=<-0.0489 -0.0252 0.1892 -0.9804>}
 ring_02_l {parent=ring_01_l, loc=<3.6174 0.4647 -0.1027>, rot=<0 0 0.2806 -0.9598>}
 ring_03_l {parent=ring_02_l, loc=<2.4152 0.114 0.0199>, rot=<0 0 0.1189 -0.9929>}
 pinky_metacarpal_l {parent=hand_l, loc=<1.7616 -0.9753 3.3814>, rot=<-0.0397 0.0814 -0.069 -0.9935>}
 pinky_01_l {parent=pinky_metacarpal_l, loc=<7.1709 -0.4831 -0.1094>, rot=<-0.0654 -0.0035 0.2242 -0.9723>}
 pinky_02_l {parent=pinky_01_l, loc=<2.8274 0.0893 -0.1043>, rot=<0.0037 0.0011 0.2854 -0.9584>}
 pinky_03_l {parent=pinky_02_l, loc=<2.2046 0.0307 -0.0078>, rot=<-0.0002 0.0098 0.1445 -0.9895>}
 deform_wrist_dn_base_l {parent=hand_l, loc=<0 -1.5326 0>, rot=<0 0 0 -1>}
 deform_wrist_dn_l {parent=deform_wrist_dn_base_l, loc=<0 0 0>, rot=<0 0 0 1>}
 deform_wrist_up_base_l {parent=hand_l, loc=<0 2.3009 0>, rot=<0 0 0 -1>}
 deform_wrist_up_l {parent=deform_wrist_up_base_l, loc=<0 0 0>, rot=<0 0 0 1>}
 lowerarm_twist_02_l {parent=lowerarm_l, loc=<10.6707 0 0>, rot=<0 0 0 1>}
 lowerarm_twist_01_l {parent=lowerarm_l, loc=<18.316 0 0>, rot=<0 0 0 1>}
 deform_elbow_in_base_l {parent=lowerarm_l, loc=<0.1834 3.6062 -0.2572>, rot=<0 0 0 -1>}
 deform_elbow_in_l {parent=deform_elbow_in_base_l, loc=<0 0 0>, rot=<0 0 0 1>}
 deform_elbow_out_base_l {parent=lowerarm_l, loc=<0 -4 0>, rot=<0 0 0 1>}
 deform_elbow_out_l {parent=deform_elbow_out_base_l, loc=<0 0 0>, rot=<0 0 0 1>}
 upperarm_twist_01_l {parent=upperarm_l, loc=<7.2135 -0.1134 0>, rot=<0 0 0 1>}
 deform_upperarm_bk_base_l {parent=upperarm_twist_01_l, loc=<-6.4053 -4.3746 0.314>, rot=<0 0 0 -1>}
 deform_upperarm_bk_l {parent=deform_upperarm_bk_base_l, loc=<0 0 0>, rot=<0 0 0 1>}
 deform_upperarm_fr_base_l {parent=upperarm_twist_01_l, loc=<-5.6312 5.092 0.3354>, rot=<0 0 0 -1>}
 deform_upperarm_fr_l {parent=deform_upperarm_fr_base_l, loc=<0 0 0>, rot=<0 0 0 1>}
 deform_upperarm_up_base_l {parent=upperarm_twist_01_l, loc=<-8.891 0.4825 3.9517>, rot=<0 0 0 -1>}
 deform_upperarm_up_l {parent=deform_upperarm_up_base_l, loc=<0 0 0>, rot=<0 0 0 1>}
 upperarm_twist_02_l {parent=upperarm_l, loc=<19.7998 -0.1134 0>, rot=<0 0 0 1>}
 dyn_scarf_MSTR {parent=clavicle_l, loc=<4.0199 -5.8308 0.9148>, rot=<-0.516 -0.5902 0.4597 -0.4171>}
 dyn_scarf_b_01 {parent=dyn_scarf_MSTR, loc=<3.7184 -1.5704 2.528>, rot=<0.0258 0.2178 -0.1313 -0.9668>}
 dyn_scarf_b_02 {parent=dyn_scarf_b_01, loc=<10.0444 0 0>, rot=<-0.0029 -0.0202 0.0118 -0.9997>}
 dyn_scarf_b_03 {parent=dyn_scarf_b_02, loc=<14.852 0 0>, rot=<-0.0066 -0.0153 -0.0062 -0.9998>}
 dyn_scarf_b_04 {parent=dyn_scarf_b_03, loc=<10.7755 0 0>, rot=<0 -0.0036 0.0133 -0.9999>}
 dyn_scarf_a_01 {parent=dyn_scarf_MSTR, loc=<3.2376 3.0848 1.5591>, rot=<0.1504 0.2501 -0.1504 -0.9445>}
 dyn_scarf_a_02 {parent=dyn_scarf_a_01, loc=<10.0535 0 0>, rot=<-0.0612 -0.0316 0.0065 -0.9976>}
 dyn_scarf_a_03 {parent=dyn_scarf_a_02, loc=<14.8569 0 0>, rot=<-0.017 -0.0065 -0.0069 -0.9998>}
 dyn_scarf_a_04 {parent=dyn_scarf_a_03, loc=<10.7789 0 0>, rot=<-0.0476 -0.0116 0.0128 -0.9987>}
 clavicle_r {parent=spine_05, loc=<5.4831 -0.1053 3.2946>, rot=<0.7289 0.0465 -0.6829 0.009>}
 upperarm_r {parent=clavicle_r, loc=<-9.8658 1.2136 0.4105>, rot=<0.0164 -0.3245 -0.0269 -0.9454>}
 lowerarm_r {parent=upperarm_r, loc=<-26.465 0.8345 -0.259>, rot=<0.0036 0.0161 -0.2298 -0.9731>}
 hand_r {parent=lowerarm_r, loc=<-22.7355 0.3203 0.6106>, rot=<-0.6784 -0.0223 0.0735 -0.7307>}
 thumb_01_r {parent=hand_r, loc=<-1.3118 0.3191 2.459>, rot=<0.6546 -0.4042 0.0664 -0.6354>}
 thumb_02_r {parent=thumb_01_r, loc=<-5.9336 0.0448 -0.0401>, rot=<-0.0037 0.0213 0.298 -0.9543>}
 thumb_03_r {parent=thumb_02_r, loc=<-2.2223 -0.3682 0.0907>, rot=<0.0118 -0.013 0.0066 -0.9998>}
 index_metacarpal_r {parent=hand_r, loc=<-1.6535 -0.0183 2.96>, rot=<0.0158 -0.0124 -0.0173 -0.9996>}
 index_01_r {parent=index_metacarpal_r, loc=<-8.3008 -0.0357 0.0524>, rot=<-0.0044 0.0485 0.0924 -0.9945>}
 index_02_r {parent=index_01_r, loc=<-3.3587 0.3004 -0.0452>, rot=<-0.002 0.0027 0.3225 -0.9466>}
 index_03_r {parent=index_02_r, loc=<-2.5578 0.123 0.0252>, rot=<-0.002 0.0067 0.2558 -0.9667>}
 middle_metacarpal_r {parent=hand_r, loc=<-2.1856 0.1725 0.5984>, rot=<0.006 0.0168 -0.0329 -0.9993>}
 middle_01_r {parent=middle_metacarpal_r, loc=<-7.3275 0.2447 -0.0129>, rot=<-0.0208 0.0042 0.0858 -0.9961>}
 middle_02_r {parent=middle_01_r, loc=<-4.0123 -0.1401 -0.1175>, rot=<-0.0218 0.0258 0.3512 -0.9357>}
 middle_03_r {parent=middle_02_r, loc=<-2.6077 -0.0534 -0.1373>, rot=<0.0082 0.0188 0.1708 -0.9851>}
 ring_metacarpal_r {parent=hand_r, loc=<-2.2074 0.3625 -1.3638>, rot=<-0.0284 0.0722 -0.0458 -0.9959>}
 ring_01_r {parent=ring_metacarpal_r, loc=<-7.0295 0.6197 0.1446>, rot=<-0.0489 -0.0252 0.1892 -0.9804>}
 ring_02_r {parent=ring_01_r, loc=<-3.6174 -0.4647 0.1027>, rot=<0 0 0.2806 -0.9598>}
 ring_03_r {parent=ring_02_r, loc=<-2.4152 -0.114 -0.0199>, rot=<0 0 0.1189 -0.9929>}
 pinky_metacarpal_r {parent=hand_r, loc=<-1.7616 0.9753 -3.3814>, rot=<-0.0397 0.0814 -0.069 -0.9935>}
 pinky_01_r {parent=pinky_metacarpal_r, loc=<-7.1709 0.4831 0.1094>, rot=<-0.0654 -0.0035 0.2242 -0.9723>}
 pinky_02_r {parent=pinky_01_r, loc=<-2.8274 -0.0893 0.1043>, rot=<0.0037 0.0011 0.2854 -0.9584>}
 pinky_03_r {parent=pinky_02_r, loc=<-2.2046 -0.0307 0.0078>, rot=<-0.0002 0.0098 0.1445 -0.9895>}
 deform_wrist_dn_base_r {parent=hand_r, loc=<-0.0001 1.5328 0>, rot=<0 0 0 -1>}
 deform_wrist_dn_r {parent=deform_wrist_dn_base_r, loc=<0 0 0>, rot=<0 0 0 1>}
 deform_wrist_up_base_r {parent=hand_r, loc=<-0.0002 -2.3007 -0.0001>, rot=<0 0 0 -1>}
 deform_wrist_up_r {parent=deform_wrist_up_base_r, loc=<0 0 0>, rot=<0 0 0 1>}
 lowerarm_twist_02_r {parent=lowerarm_r, loc=<-10.6707 0 0>, rot=<0 0 0 -1>}
 lowerarm_twist_01_r {parent=lowerarm_r, loc=<-18.316 0 0>, rot=<0 0 0 -1>}
 deform_elbow_in_base_r {parent=lowerarm_r, loc=<-0.1834 -3.6062 0.2572>, rot=<0 0 0 -1>}
 deform_elbow_in_r {parent=deform_elbow_in_base_r, loc=<0 0 0>, rot=<0 0 0 1>}
 deform_elbow_out_base_r {parent=lowerarm_r, loc=<0 4 -0.0001>, rot=<0 0 0 -1>}
 deform_elbow_out_r {parent=deform_elbow_out_base_r, loc=<0 0 0>, rot=<0 0 0 1>}
 upperarm_twist_01_r {parent=upperarm_r, loc=<-7.2135 0.1134 0>, rot=<0 0 0 -1>}
 deform_upperarm_bk_base_r {parent=upperarm_twist_01_r, loc=<6.4053 4.3746 -0.314>, rot=<0 0 0 -1>}
 deform_upperarm_bk_r {parent=deform_upperarm_bk_base_r, loc=<0 0 0>, rot=<0 0 0 1>}
 deform_upperarm_fr_base_r {parent=upperarm_twist_01_r, loc=<5.6312 -5.092 -0.3354>, rot=<0 0 0 -1>}
 deform_upperarm_fr_r {parent=deform_upperarm_fr_base_r, loc=<0 0 0>, rot=<0 0 0 1>}
 deform_upperarm_up_base_r {parent=upperarm_twist_01_r, loc=<8.891 -0.4825 -3.9517>, rot=<0 0 0 -1>}
 deform_upperarm_up_r {parent=deform_upperarm_up_base_r, loc=<0 0 0>, rot=<0 0 0 1>}
 upperarm_twist_02_r {parent=upperarm_r, loc=<-19.7998 0.1134 0>, rot=<0 0 0 -1>}
 deform_pec_aim_base_l {parent=spine_04, loc=<3.2456 9.3309 -1.2237>, rot=<0.0603 -0.4033 0.3318 -0.8507>}
 deform_pec_aim_l {parent=deform_pec_aim_base_l, loc=<0 0 0>, rot=<0 0 0 1>}
 deform_pec_base_l {parent=deform_pec_aim_l, loc=<9.3665 0 0>, rot=<0 0 0 1>}
 deform_pec_l {parent=deform_pec_base_l, loc=<0 0 0>, rot=<0 0 0 1>}
 deform_pec_aim_base_r {parent=spine_04, loc=<3.2452 9.3309 1.2237>, rot=<-0.4033 -0.0603 0.8507 0.3318>}
 deform_pec_aim_r {parent=deform_pec_aim_base_r, loc=<0 0 0>, rot=<0 0 0 1>}
 deform_pec_base_r {parent=deform_pec_aim_r, loc=<-9.3667 -0.0003 -0.0004>, rot=<0 0 0 1>}
 deform_pec_r {parent=deform_pec_base_r, loc=<0 0 0>, rot=<0 0 0 1>}
 thigh_l {parent=pelvis, loc=<-1.0855 -0.2109 -8.7053>, rot=<0.0009 0.0355 0.0155 -0.9993>}
 calf_l {parent=thigh_l, loc=<-42.5284 -0.0019 -0.7297>, rot=<0.0006 -0.0161 -0.0391 -0.9991>}
 foot_l {parent=calf_l, loc=<-46.7401 -2.5014 -0.592>, rot=<0.0427 -0.0238 0.0353 -0.9982>}
 ball_l {parent=foot_l, loc=<-3.5658 10.8066 -0.1544>, rot=<0.0001 -0.0001 -0.7063 -0.7079>}
 calf_twist_02_l {parent=calf_l, loc=<-19.1546 -1.7943 -0.2638>, rot=<0.0237 0.0095 -0.0035 -0.9997>}
 calf_twist_01_l {parent=calf_l, loc=<-35.8483 -1.9197 -0.5794>, rot=<0.0237 0.0095 -0.0035 -0.9997>}
 deform_knee_in_base_l {parent=calf_l, loc=<0.8031 -5.0231 0.2456>, rot=<0 0 0 -1>}
 deform_knee_in_l {parent=deform_knee_in_base_l, loc=<0 0 0>, rot=<0 0 0 1>}
 deform_knee_out_base_l {parent=calf_l, loc=<0 4 0>, rot=<0 0 0 -1>}
 deform_knee_out_l {parent=deform_knee_out_base_l, loc=<0 0 0>, rot=<0 0 0 1>}
 thigh_twist_01_l {parent=thigh_l, loc=<-27.952 0 0>, rot=<0.0013 0 0 -1>}
 deform_groin_base_l {parent=thigh_l, loc=<0.4801 5.6776 -0.2486>, rot=<0 0 0 -1>}
 deform_groin_l {parent=deform_groin_base_l, loc=<0 0 0>, rot=<0 0 0 1>}
 thigh_r {parent=pelvis, loc=<-1.0853 -0.2109 8.7053>, rot=<0.0355 -0.0009 0.9993 0.0155>}
 calf_r {parent=thigh_r, loc=<42.5284 0.0019 0.7297>, rot=<0.0006 -0.0161 -0.0391 -0.9991>}
 foot_r {parent=calf_r, loc=<46.7401 2.5014 0.592>, rot=<0.0427 -0.0238 0.0353 -0.9982>}
 ball_r {parent=foot_r, loc=<3.5658 -10.8066 0.1544>, rot=<0.0001 -0.0001 -0.7063 -0.7079>}
 calf_twist_02_r {parent=calf_r, loc=<19.1546 1.7943 0.2638>, rot=<0.0237 0.0095 -0.0035 -0.9997>}
 calf_twist_01_r {parent=calf_r, loc=<35.8483 1.9197 0.5794>, rot=<0.0237 0.0095 -0.0035 -0.9997>}
 deform_knee_in_base_r {parent=calf_r, loc=<-0.8031 5.0231 -0.2457>, rot=<0 0 0 -1>}
 deform_knee_in_r {parent=deform_knee_in_base_r, loc=<0 0 0>, rot=<0 0 0 1>}
 deform_knee_out_base_r {parent=calf_r, loc=<0 -4 0>, rot=<0 0 0 -1>}
 deform_knee_out_r {parent=deform_knee_out_base_r, loc=<0 0 0>, rot=<0 0 0 1>}
 thigh_twist_01_r {parent=thigh_r, loc=<27.952 0 0>, rot=<0.0013 0 0 -1>}
 deform_groin_base_r {parent=thigh_r, loc=<-0.4801 -5.6776 0.2486>, rot=<0 0 0 -1>}
 deform_groin_r {parent=deform_groin_base_r, loc=<0 0 0>, rot=<0 0 0 1>}
 dyn_skirtFront_01_r {parent=thigh_r, loc=<3.6133 -5.7038 -4.511>, rot=<-0.707 0.0057 0.0209 -0.7069>}
 dyn_skirtFront_02_r {parent=dyn_skirtFront_01_r, loc=<6.6314 0 0>, rot=<0.0005 0.0003 0.0141 -0.9999>}
 deform_glute_base_l {parent=pelvis, loc=<-7.3466 -9.0213 -9.1245>, rot=<0.0172 0.031 -0.4477 -0.8935>}
 deform_glute_l {parent=deform_glute_base_l, loc=<0 0 0>, rot=<0 0 0 1>}
 deform_glute_base_r {parent=pelvis, loc=<-7.3464 -9.0213 9.1245>, rot=<0.031 -0.0172 0.8935 -0.4477>}
 deform_glute_r {parent=deform_glute_base_r, loc=<0 0 0>, rot=<0 0 0 1>}
 dyn_skirtBack_mstr_r {parent=pelvis, loc=<2.3815 -7.7391 9.4915>, rot=<0.0932 -0.4147 0.9045 -0.0343>}
 dyn_skirtBack_01_r {parent=dyn_skirtBack_mstr_r, loc=<0.0002 0 -0.0001>, rot=<0.0117 0.0256 -0.0006 -0.9996>}
 dyn_skirtBack_02_r {parent=dyn_skirtBack_01_r, loc=<10.9248 0 0>, rot=<0.0039 0.0261 0.0002 -0.9997>}
 dyn_skirtBack_03_r {parent=dyn_skirtBack_02_r, loc=<11.2672 0 0>, rot=<0 -0.0168 -0.0004 -0.9999>}
 dyn_skirtBack_04_r {parent=dyn_skirtBack_03_r, loc=<13.4212 0 0>, rot=<0 0 0 1>}
 dyn_skirtBack_mstr_l {parent=pelvis, loc=<2.4079 -7.778 -9.5237>, rot=<0.0357 -0.9051 0.4135 -0.0923>}
 dyn_skirtBack_01_l {parent=dyn_skirtBack_mstr_l, loc=<0 0 0>, rot=<0.0017 0.0258 -0.0004 -0.9997>}
 dyn_skirtBack_02_l {parent=dyn_skirtBack_01_l, loc=<10.925 0 0>, rot=<-0.005 0.0261 0.0005 -0.9996>}
 dyn_skirtBack_03_l {parent=dyn_skirtBack_02_l, loc=<11.2672 0 0>, rot=<0.0033 -0.0168 -0.0001 -0.9999>}
 dyn_skirtBack_mstr {parent=pelvis, loc=<1.4277 -10.8787 0>, rot=<0.0418 -0.7059 0.7059 -0.0418>}
 dyn_skirtBack_01 {parent=dyn_skirtBack_mstr, loc=<0 0 0>, rot=<0 0.0685 0 -0.9977>}
 dyn_skirtBack_02 {parent=dyn_skirtBack_01, loc=<10.4084 0 0>, rot=<0 0.0062 0 -1>}
 dyn_skirtBack_03 {parent=dyn_skirtBack_02, loc=<10.8984 0 0>, rot=<0 -0.0098 0 -1>}
 dyn_skirtSide_mstr_r {parent=pelvis, loc=<-5.4296 -4.8312 13.9134>, rot=<0.157 -0.0508 0.9863 0.002>}
 dyn_skirtSide_01_r {parent=dyn_skirtSide_mstr_r, loc=<0.0002 0 0>, rot=<0 0 0 -1>}
 dyn_skirtSide_02_r {parent=dyn_skirtSide_01_r, loc=<8.8494 -0.089 -0.0095>, rot=<0.0544 0.002 -0.0124 -0.9984>}
 dyn_skirtSide_03_r {parent=dyn_skirtSide_02_r, loc=<8.8483 -0.0117 -0.0084>, rot=<0.0063 0.0012 -0.0016 -1>}
 dyn_skirtSide_04_r {parent=dyn_skirtSide_03_r, loc=<8.8485 0.0192 0.026>, rot=<-0.0093 -0.0036 0.0027 -0.9999>}
 dyn_skirtSide_05_r {parent=dyn_skirtSide_04_r, loc=<8.8483 0 0>, rot=<0 0 0 -1>}
 dyn_skirtFront_mstr {parent=pelvis, loc=<-2.6188 5.8834 1.1416>, rot=<-0.0235 -0.7097 -0.7034 -0.0316>}
 dyn_skirtFront_01 {parent=dyn_skirtFront_mstr, loc=<0 0 0>, rot=<-0.0301 0.0044 0.0329 -0.999>}
 dyn_skirtFront_02 {parent=dyn_skirtFront_01, loc=<11.6638 0 0>, rot=<-0.0147 0.0009 0.0106 -0.9998>}
 dyn_skirtFront_03 {parent=dyn_skirtFront_02, loc=<12.8186 0 0>, rot=<-0.0055 -0.0129 0.0076 -0.9999>}
 dyn_skirtFront_04 {parent=dyn_skirtFront_03, loc=<11.6946 0 0>, rot=<0.0716 0.0061 0.0068 -0.9974>}
 dyn_skirtFront_05 {parent=dyn_skirtFront_04, loc=<11.3265 0 0>, rot=<0.0007 0 0 -1>}
 dyn_skirtFront_mstr_l {parent=pelvis, loc=<0.6501 4.2133 -10.2421>, rot=<-0.0237 -0.8406 -0.5354 -0.0779>}
 dyn_skirtFront_01_l {parent=dyn_skirtFront_mstr_l, loc=<0 0 0>, rot=<0.0072 0 0 -1>}
 dyn_skirtFront_02_l {parent=dyn_skirtFront_01_l, loc=<10.7807 0 0>, rot=<-0.0355 0.0019 0.0151 -0.9993>}
 dyn_skirtFront_03_l {parent=dyn_skirtFront_02_l, loc=<11.1533 0 0>, rot=<-0.0025 -0.0086 0.0066 -0.9999>}
 dyn_skirtFront_04_l {parent=dyn_skirtFront_03_l, loc=<14.7441 0 0>, rot=<0.0047 0 0 -1>}
 dyn_skirtSide_mstr_l1 {parent=pelvis, loc=<1.7738 -1.0185 -12.7574>, rot=<-0.0064 -0.9885 -0.046 -0.1441>}
 dyn_skirtSide_01_l {parent=dyn_skirtSide_mstr_l1, loc=<0 0 0>, rot=<0 0 0 -1>}
 dyn_skirtSide_02_l {parent=dyn_skirtSide_01_l, loc=<11.5861 0.0581 -0.0006>, rot=<0.0172 -0.0001 0.0157 -0.9997>}
 dyn_skirtSide_03_l {parent=dyn_skirtSide_02_l, loc=<11.5852 0.0067 -0.0032>, rot=<-0.0062 0.0009 0.0018 -1>}
 dyn_skirtSide_04_l {parent=dyn_skirtSide_03_l, loc=<11.5852 -0.0121 0.0115>, rot=<0.0093 -0.0031 -0.0033 -0.9999>}
 dyn_Quiver_MSTR {parent=pelvis, loc=<-5.4296 -4.8312 13.9134>, rot=<0.1999 -0.0508 0.9785 -0.0002>}
 dyn_quiver {parent=dyn_Quiver_MSTR, loc=<0 0 0>, rot=<0 0 0 -1>}
 dyn_arrow_a {parent=dyn_quiver, loc=<9.5282 18.0865 -4.4104>, rot=<0.1233 -0.1458 -0.0273 -0.9812>}
 dyn_arrow_b {parent=dyn_quiver, loc=<10.3121 17.9581 -4.2181>, rot=<0.1233 -0.1458 -0.0273 -0.9812>}
 dyn_arrow_c {parent=dyn_quiver, loc=<11.2296 18.7764 -4.8871>, rot=<0.1233 -0.1458 -0.0273 -0.9812>}
 dyn_moon_charm {parent=dyn_quiver, loc=<8.1524 9.9414 1.4353>, rot=<0.1762 -0.1995 -0.0283 -0.9635>}
 dyn_moon_charm01 {parent=dyn_moon_charm, loc=<1.6914 0 0.283>, rot=<0 0 0 1>}
 dyn_chain_01 {parent=pelvis, loc=<-3.4917 6.8997 -4.7194>, rot=<-0.294 -0.7509 0.5333 -0.2558>}
 dyn_chain_02 {parent=dyn_chain_01, loc=<3.888 0 0>, rot=<0.0505 0.2402 -0.128 -0.9609>}
 dyn_chain_03 {parent=dyn_chain_02, loc=<2.6952 0 0>, rot=<0.0045 0.0241 -0.0194 -0.9995>}
 dyn_chain_04 {parent=dyn_chain_03, loc=<2.7845 0 0>, rot=<-0.0041 0.1846 0.0291 -0.9824>}
 dyn_chain_05 {parent=dyn_chain_04, loc=<2.8263 0 0>, rot=<0.0004 0.1037 0.0174 -0.9945>}
 dyn_chain_06 {parent=dyn_chain_05, loc=<2.6832 0 0>, rot=<0.0108 0.22 0.0456 -0.9744>}
 dyn_chain_07 {parent=dyn_chain_06, loc=<2.7693 0 0>, rot=<0.015 0.1307 0.0296 -0.9909>}
 dyn_chain_08 {parent=dyn_chain_07, loc=<2.7831 0 0>, rot=<0.0747 0.1717 0.0964 -0.9776>}
 dyn_chain_09 {parent=dyn_chain_08, loc=<2.781 0 0>, rot=<0.0184 0.1345 0.0157 -0.9906>}
 dyn_chain_10 {parent=dyn_chain_09, loc=<2.9701 0 0>, rot=<0.1102 0.0481 0.0797 -0.9895>}
 dyn_chain_11 {parent=dyn_chain_10, loc=<2.4774 0 0>, rot=<0.098 0.0762 0.0671 -0.99>}
 dyn_chain_12 {parent=dyn_chain_11, loc=<2.6003 0 0>, rot=<0.0829 0.1688 0.0418 -0.9813>}
 dyn_chain_13 {parent=dyn_chain_12, loc=<2.8788 0 0>, rot=<-0.0905 0.1221 -0.025 -0.9881>}
 dyn_chain_14 {parent=dyn_chain_13, loc=<2.8394 0 0>, rot=<0.1393 0.0379 0.0252 -0.9892>}
 dyn_chain_15 {parent=dyn_chain_14, loc=<2.7953 0 0>, rot=<0 0.1313 0 -0.9913>}
 dyn_chain_16 {parent=dyn_chain_15, loc=<3.191 0 0>, rot=<0 0 0 1>}
 dyn_chain_01_ctrl {controlType=EULER_TRANSFORM, parent=pelvis, type=Control, loc=<4.7194 6.24 94.3205>, rot=<0.5721 0.347 -0.1754 0.7222>}
 dyn_chain_04_ctrl {controlType=EULER_TRANSFORM, parent=pelvis, type=Control, loc=<12.4255 5.4829 89.7662>, rot=<0.4658 -0.075 -0.307 0.8265>}
 dyn_chain_07_ctrl {controlType=EULER_TRANSFORM, parent=pelvis, type=Control, loc=<17.44 -0.6916 89.4361>, rot=<0.2688 -0.3794 -0.5585 0.6869>}
 dyn_chain_12_ctrl {controlType=EULER_TRANSFORM, parent=pelvis, type=Control, loc=<11.8193 -11.2925 93.831>, rot=<-0.2303 -0.3447 -0.8822 0.2235>}
 dyn_chain_16_ctrl {controlType=EULER_TRANSFORM, parent=pelvis, type=Control, loc=<3.9722 -11.4942 102.1555>, rot=<0.4792 0.3563 0.7969 -0.0911>}
 dyn_chain_frt_MSTR {parent=pelvis, loc=<-3.4917 6.8997 -4.7194>, rot=<0.0504 -0.7494 -0.5997 -0.2759>}
 dyn_chain_frt_ctrl {parent=dyn_chain_frt_MSTR, loc=<0 0 0>, rot=<0 0 0 1>}
 dyn_chain_bk_MSTR {parent=pelvis, loc=<3.703 -11.1038 -3.9722>, rot=<0.0708 -0.7972 0.5728 -0.1772>}
 dyn_chain_bk_ctrl {parent=dyn_chain_bk_MSTR, loc=<0 0 0>, rot=<0 0 0 1>}
 dyn_chain_mid_MSTR {parent=pelvis, loc=<-8.6213 0.1475 -17.44>, rot=<0.2009 -0.2211 0.5984 -0.7434>}
 dyn_chain_mid_ctrl {parent=dyn_chain_mid_MSTR, loc=<0 0 0>, rot=<0 0 0 1>}
 dyn_chain_bk_md_MSTR {parent=pelvis, loc=<-4.6089 -10.604 -11.8193>, rot=<0.788 0.0716 0.4681 -0.3934>}
 dyn_chain_bk_md_ctrl {parent=dyn_chain_bk_md_MSTR, loc=<0 0 0>, rot=<0 0 0 1>}
 dyn_chain_frt_md_MSTR {parent=pelvis, loc=<-8.0703 6.3061 -12.4255>, rot=<-0.1218 -0.5293 0.5578 -0.6276>}
 dyn_chain_frt_md_ctrl {parent=dyn_chain_frt_md_MSTR, loc=<0 0 0>, rot=<0 0 0 1>}
 Floor_COL {parent=root, loc=<0 0 0>, rot=<0 0 0 1>}
]
variables [
 FrontendIdle {type=double, default=0}
 Frt_constraint L {type=double, default=0}
 Frt_constraint R {type=double, default=0}
 Moving2d {type=bool, default=False}
 jumping {type=double, default=0}
 jumping_chain {type=double, default=0}
 moving2d alpha {type=double, default=0}
 thigh l bk {type=double, default=0}
 thigh l in {type=double, default=0}
 thigh l side {type=double, default=0}
 thigh l up {type=double, default=0}
 thigh r bk {type=double, default=0}
 thigh r in {type=double, default=0}
 thigh r side {type=double, default=0}
 thigh r up {type=double, default=0}
]
