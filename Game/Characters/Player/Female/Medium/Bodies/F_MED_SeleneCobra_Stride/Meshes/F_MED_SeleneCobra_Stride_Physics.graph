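# Physics asset for the Selene Cobra Stride medium female player body, covering secondary motion of scarf, skirt and quiver.
# Capsules cover torso, arms and legs (no head, hands or feet), plus two spheres on Floor_COL, apparently a floor proxy.
# Skirt chains and quiver hang from the pelvis on free angular constraints.
# Two scarf chains branch from a left-clavicle master with tight twist limits, cross-linked at segments 02 and 03.
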PhysicsAsset /Game/Characters/Player/Female/Medium/Bodies/F_MED_SeleneCobra_Stride/Meshes/F_MED_SeleneCobra_Stride_Physics
bodies {
 Floor_COL [
  {shapeType=Sphere, radius=212, shapeName=floorCOL_sphere, center=<-3.4923 -2.714 -209>}
  {shapeType=Sphere, radius=27.2085, shapeName=floorCOL_sphere, center=<0 2.435 -0.4285>}
 ]
 calf_l [{shapeType=Capsule, length=38.7928, radius=8.1658, center=<-23.4523 -1.1162 0.1776>, rotation=<rot 0.1246 84.5345 -88.5757>}]
 calf_r [{shapeType=Capsule, length=39.9107, radius=8.7978, center=<26.3559 0.7114 -0.6693>, rotation=<rot -0.3385 89.5329 -88.4634>}]
 clavicle_l [{
   shapeType=Capsule
   length=5.9581
   radius=5.3397
   shapeName=clavicle_l_capsule
   center=<7.4519 -0.7392 0.104>
   rotation=<rot 16.9211 87.287 -90>
  }]
 clavicle_r [{
   shapeType=Capsule
   length=3.8659
   radius=5.6487
   shapeName=clavicle_r_capsule
   center=<-6.1046 0.3704 0.8599>
   rotation=<rot 3.2384 98.0696 -90>
  }]
 dyn_quiver [{
   shapeType=Capsule
   length=30.4337
   radius=6.4773
   shapeName=dyn_quiver_capsule
   center=<6.2467 11.2291 2.3282>
   rotation=<rot 0 -21.0535 -70.7614>
  }]
 dyn_scarf_MSTR [{shapeType=Capsule, length=0.505, radius=0.505, shapeName=dyn_scarf_MSTR_capsule}]
 dyn_scarf_a_01 [{shapeType=Capsule, length=12.1875, radius=5.625, shapeName=dyn_scarf_a_01_capsule, center=<5 0 4>, rotation=<rot 90 -11.3779 -11.378>}]
 dyn_scarf_a_02 [{shapeType=Capsule, length=11.1875, radius=6.625, shapeName=dyn_scarf_a_02_capsule, center=<5 0 3>, rotation=<rot 90 -9.6284 -9.6285>}]
 dyn_scarf_a_03 [{shapeType=Capsule, length=10.75, radius=7.2812, shapeName=dyn_scarf_a_03_capsule, center=<5 0 3>, rotation=<rot 90 -10.2755 -10.2756>}]
 dyn_scarf_a_04 [{shapeType=Capsule, length=12.1562, radius=5.5, shapeName=dyn_scarf_a_04_capsule, center=<5 0 5>, rotation=<rot 90 -8.6432 -8.6432>}]
 dyn_scarf_b_01 [{shapeType=Capsule, length=12.1875, radius=5.625, shapeName=dyn_scarf_b_01_capsule, center=<5 0 4>, rotation=<rot 90 -5.2469 -5.2469>}]
 dyn_scarf_b_02 [{shapeType=Capsule, length=11.1875, radius=6.625, shapeName=dyn_scarf_b_02_capsule, center=<5 0 3>, rotation=<rot 90 -3.8377 -3.8377>}]
 dyn_scarf_b_03 [{shapeType=Capsule, length=10.75, radius=7.2812, shapeName=dyn_scarf_b_03_capsule, center=<5 0 3>, rotation=<rot 90 -4.514 -4.514>}]
 dyn_scarf_b_04 [{shapeType=Capsule, length=15.2812, radius=5.5, shapeName=dyn_scarf_b_04_capsule, center=<8 0 5>, rotation=<rot 90 -2.9801 -2.9801>}]
 dyn_skirtBack_01 [{shapeType=Capsule, length=8.2188, radius=8.9062, shapeName=dyn_skirtBack_01_capsule, center=<7 0 -6>, rotation=<rot 90 0 0>}]
 dyn_skirtBack_01_l [{shapeType=Capsule, length=12.0625, radius=9, shapeName=dyn_skirtBack_01_l_capsule, center=<7 0 -6>, rotation=<rot 90 0 0>}]
 dyn_skirtBack_01_r [{shapeType=Capsule, length=9.125, radius=8, shapeName=dyn_skirtBack_01_r_capsule, center=<7 0 -6>, rotation=<rot 90 0 0>}]
 dyn_skirtBack_02 [{shapeType=Capsule, length=8.2188, radius=8.9062, shapeName=dyn_skirtBack_02_capsule, center=<7 0 -6>, rotation=<rot 90 0 0>}]
 dyn_skirtBack_02_l [{shapeType=Capsule, length=12.0625, radius=9, shapeName=dyn_skirtBack_02_l_capsule, center=<7 0 -6>, rotation=<rot 90 0 0>}]
 dyn_skirtBack_02_r [{shapeType=Capsule, length=9.125, radius=8, shapeName=dyn_skirtBack_02_r_capsule, center=<7 0 -6>, rotation=<rot 90 0 0>}]
 dyn_skirtBack_03 [{shapeType=Capsule, length=8.2188, radius=8.9062, shapeName=dyn_skirtBack_03_capsule, center=<8 0 -6>, rotation=<rot 90 0 0>}]
 dyn_skirtBack_03_l [{shapeType=Capsule, length=12.0625, radius=9, shapeName=dyn_skirtBack_03_l_capsule, center=<8 0 -6>, rotation=<rot 90 0 0>}]
 dyn_skirtBack_03_r [{shapeType=Capsule, length=9.125, radius=8, shapeName=dyn_skirtBack_03_r_capsule, center=<7 0 -6>, rotation=<rot 90 0 0>}]
 dyn_skirtBack_04_r [{shapeType=Capsule, length=9.125, radius=8, shapeName=dyn_skirtBack_04_r_capsule, center=<7 0 -8>, rotation=<rot 0 0 85>}]
 dyn_skirtFront_01_l [{shapeType=Capsule, length=8.25, radius=8.875, shapeName=dyn_skirtSide_01_l_capsule, center=<7 1.7489 -4.662>, rotation=<rot 90 0 0>}]
 dyn_skirtFront_02_l [{shapeType=Capsule, length=8.25, radius=8.875, shapeName=dyn_skirtSide_02_l_capsule, center=<7 1.7489 -4.662>, rotation=<rot 90 0 0>}]
 dyn_skirtSide_01_l [{shapeType=Capsule, length=7.5, radius=9.625, shapeName=dyn_skirtSide_01_l_capsule, center=<7 0.707 -7>, rotation=<rot 90 0 0>}]
 dyn_skirtSide_02_l [{shapeType=Capsule, length=7.5, radius=9.625, shapeName=dyn_skirtSide_02_l_capsule, center=<7 0.707 -7>, rotation=<rot 90 0 0>}]
 dyn_skirtSide_03_l [{shapeType=Capsule, length=12.375, radius=9.625, shapeName=dyn_skirtSide_03_l_capsule, center=<7 0 -6>, rotation=<rot 0 0 90>}]
 dyn_skirtSide_04_l [{shapeType=Capsule, length=12.125, radius=9.875, shapeName=dyn_skirtSide_04_l_capsule, center=<7 0 -6>, rotation=<rot 0 0 90>}]
 lowerarm_l [{shapeType=Capsule, length=10.8099, radius=7.3667, center=<9.8537 -0.4147 0.4756>, rotation=<rot 0 -93.9158 -88.4307>}]
 lowerarm_r [{shapeType=Capsule, length=13.7095, radius=4.697, center=<-9.8803 -0.0664 0.2805>, rotation=<rot 0 -84.9882 -89.3504>}]
 pelvis [
  {shapeType=Capsule, length=9.3375, radius=11.9301, center=<0.5724 -2.2291 0>}
  {shapeType=Capsule, length=12.665, radius=8.9702, shapeName=pelvis_capsule, center=<-3.9978 -4.3248 0>}
 ]
 spine_01 [{shapeType=Capsule, length=10.38, radius=10.755, shapeName=spine_01_capsule, center=<0.7809 -1.3783 0>}]
 spine_02 [{shapeType=Capsule, length=0.375, radius=16.2122, shapeName=spine_02_capsule, center=<-2.0211 9.9587 0>, rotation=<rot -90 0 -0.0004>}]
 spine_03 [{shapeType=Capsule, length=3.505, radius=10.63, shapeName=spine_03_capsule, center=<4.3158 1.8107 0>}]
 spine_04 [{shapeType=Capsule, length=8.5, radius=10.2595, shapeName=spine_04_capsule, center=<3.4229 1.6089 0>}]
 spine_05 [{shapeType=Capsule, length=9.4751, radius=7.3851, shapeName=spine_05_capsule, center=<1.0843 -0.2726 0>}]
 thigh_l [{shapeType=Capsule, length=23.8152, radius=12.1848, center=<-22.9661 -5.695 0.5983>, rotation=<rot 76.6232 -173.0883 175.7852>}]
 thigh_r [{shapeType=Capsule, length=27.6902, radius=8.3098, center=<21.0533 0.5413 -1.1849>, rotation=<rot 85 -180 -180>}]
 upperarm_l [{shapeType=Capsule, length=17.8604, radius=7.2054, shapeName=upperarm_l_capsule, center=<13.3624 -1.25 0.0001>, rotation=<rot 0 -90 90>}]
 upperarm_r [{shapeType=Capsule, length=18.9229, radius=4.5179, shapeName=upperarm_r_capsule, center=<-12.1286 1.25 0.0001>, rotation=<rot 0 -90 90>}]
}
constraints [
 {bone1Name=spine_04, bone2Name=spine_03, swing1LimitDegrees=45, swing2LimitDegrees=45, twistLimitDegrees=45}
 {bone1Name=spine_05, bone2Name=spine_04, swing1LimitDegrees=45, swing2LimitDegrees=45, twistLimitDegrees=45}
 {bone1Name=clavicle_l, bone2Name=spine_05, swing1LimitDegrees=45, swing2LimitDegrees=45, twistLimitDegrees=45}
 {bone1Name=lowerarm_l, bone2Name=upperarm_l, swing1LimitDegrees=45, swing2LimitDegrees=45, twistLimitDegrees=45}
 {bone1Name=clavicle_r, bone2Name=spine_05, swing1LimitDegrees=45, swing2LimitDegrees=45, twistLimitDegrees=45}
 {bone1Name=lowerarm_r, bone2Name=upperarm_r, swing1LimitDegrees=45, swing2LimitDegrees=45, twistLimitDegrees=45}
 {bone1Name=thigh_l, bone2Name=pelvis, swing1LimitDegrees=45, swing2LimitDegrees=45, twistLimitDegrees=45}
 {bone1Name=calf_l, bone2Name=thigh_l, swing1LimitDegrees=45, swing2LimitDegrees=45, twistLimitDegrees=45}
 {bone1Name=thigh_r, bone2Name=pelvis, swing1LimitDegrees=45, swing2LimitDegrees=45, twistLimitDegrees=45}
 {bone1Name=calf_r, bone2Name=thigh_r, swing1LimitDegrees=45, swing2LimitDegrees=45, twistLimitDegrees=45}
 {
  bone1Name=spine_03
  bone2Name=spine_02
  swing1LimitDegrees=45
  swing1Motion=Free
  swing2LimitDegrees=45
  swing2Motion=Free
  twistLimitDegrees=45
  twistMotion=Free
 }
 {
  bone1Name=spine_01
  bone2Name=pelvis
  swing1LimitDegrees=45
  swing1Motion=Free
  swing2LimitDegrees=45
  swing2Motion=Free
  twistLimitDegrees=45
  twistMotion=Free
 }
 {
  bone1Name=spine_02
  bone2Name=spine_01
  swing1LimitDegrees=45
  swing1Motion=Free
  swing2LimitDegrees=45
  swing2Motion=Free
  twistLimitDegrees=45
  twistMotion=Free
 }
 {
  bone1Name=upperarm_r
  bone2Name=clavicle_r
  swing1LimitDegrees=45
  swing1Motion=Free
  swing2LimitDegrees=45
  swing2Motion=Free
  twistLimitDegrees=45
  twistMotion=Free
 }
 {
  bone1Name=upperarm_l
  bone2Name=clavicle_l
  swing1LimitDegrees=45
  swing1Motion=Free
  swing2LimitDegrees=45
  swing2Motion=Free
  twistLimitDegrees=45
  twistMotion=Free
 }
 {
  bone1Name=dyn_skirtBack_01
  bone2Name=pelvis
  swing1LimitDegrees=45
  swing1Motion=Free
  swing2LimitDegrees=45
  swing2Motion=Free
  twistLimitDegrees=45
  twistMotion=Free
 }
 {
  bone1Name=dyn_skirtBack_02
  bone2Name=dyn_skirtBack_01
  swing1LimitDegrees=45
  swing1Motion=Free
  swing2LimitDegrees=45
  swing2Motion=Free
  twistLimitDegrees=45
  twistMotion=Free
 }
 {
  bone1Name=dyn_skirtBack_03
  bone2Name=dyn_skirtBack_02
  swing1LimitDegrees=45
  swing1Motion=Free
  swing2LimitDegrees=45
  swing2Motion=Free
  twistLimitDegrees=45
  twistMotion=Free
 }
 {
  bone1Name=dyn_skirtBack_01_l
  bone2Name=pelvis
  swing1LimitDegrees=45
  swing1Motion=Free
  swing2LimitDegrees=45
  swing2Motion=Free
  twistLimitDegrees=45
  twistMotion=Free
 }
 {
  bone1Name=dyn_skirtBack_02_l
  bone2Name=dyn_skirtBack_01_l
  swing1LimitDegrees=45
  swing1Motion=Free
  swing2LimitDegrees=45
  swing2Motion=Free
  twistLimitDegrees=45
  twistMotion=Free
 }
 {
  bone1Name=dyn_skirtBack_03_l
  bone2Name=dyn_skirtBack_02_l
  swing1LimitDegrees=45
  swing1Motion=Free
  swing2LimitDegrees=45
  swing2Motion=Free
  twistLimitDegrees=45
  twistMotion=Free
 }
 {
  bone1Name=dyn_skirtBack_01_r
  bone2Name=pelvis
  swing1LimitDegrees=45
  swing1Motion=Free
  swing2LimitDegrees=45
  swing2Motion=Free
  twistLimitDegrees=45
  twistMotion=Free
 }
 {
  bone1Name=dyn_skirtBack_02_r
  bone2Name=dyn_skirtBack_01_r
  swing1LimitDegrees=45
  swing1Motion=Free
  swing2LimitDegrees=45
  swing2Motion=Free
  twistLimitDegrees=45
  twistMotion=Free
 }
 {
  bone1Name=dyn_skirtBack_03_r
  bone2Name=dyn_skirtBack_02_r
  swing1LimitDegrees=45
  swing1Motion=Free
  swing2LimitDegrees=45
  swing2Motion=Free
  twistLimitDegrees=45
  twistMotion=Free
 }
 {
  bone1Name=dyn_skirtBack_04_r
  bone2Name=dyn_skirtBack_03_r
  swing1LimitDegrees=45
  swing1Motion=Free
  swing2LimitDegrees=45
  swing2Motion=Free
  twistLimitDegrees=45
  twistMotion=Free
 }
 {
  bone1Name=dyn_skirtSide_04_l
  bone2Name=dyn_skirtSide_03_l
  swing1LimitDegrees=45
  swing1Motion=Free
  swing2LimitDegrees=45
  swing2Motion=Free
  twistLimitDegrees=45
  twistMotion=Free
 }
 {
  bone1Name=dyn_skirtSide_03_l
  bone2Name=dyn_skirtSide_02_l
  swing1LimitDegrees=45
  swing1Motion=Free
  swing2LimitDegrees=45
  swing2Motion=Free
  twistLimitDegrees=45
  twistMotion=Free
 }
 {
  bone1Name=dyn_skirtSide_02_l
  bone2Name=dyn_skirtSide_01_l
  swing1LimitDegrees=45
  swing1Motion=Free
  swing2LimitDegrees=45
  swing2Motion=Free
  twistLimitDegrees=45
  twistMotion=Free
 }
 {
  bone1Name=dyn_skirtSide_01_l
  bone2Name=pelvis
  swing1LimitDegrees=45
  swing1Motion=Free
  swing2LimitDegrees=45
  swing2Motion=Free
  twistLimitDegrees=45
  twistMotion=Free
 }
 {bone1Name=dyn_scarf_b_04, bone2Name=dyn_scarf_b_03, swing1LimitDegrees=35, swing2LimitDegrees=45, twistLimitDegrees=3}
 {
  bone1Name=dyn_scarf_MSTR
  bone2Name=clavicle_l
  swing1LimitDegrees=45
  swing1Motion=Free
  swing2LimitDegrees=45
  swing2Motion=Free
  twistLimitDegrees=45
  twistMotion=Free
 }
 {bone1Name=dyn_scarf_a_01, bone2Name=dyn_scarf_MSTR, swing1LimitDegrees=15, swing2LimitDegrees=35, twistLimitDegrees=2}
 {bone1Name=dyn_scarf_a_02, bone2Name=dyn_scarf_a_01, swing1LimitDegrees=35, swing2LimitDegrees=45, twistLimitDegrees=2}
 {bone1Name=dyn_scarf_a_03, bone2Name=dyn_scarf_a_02, swing1LimitDegrees=35, swing2LimitDegrees=45, twistLimitDegrees=3}
 {bone1Name=dyn_scarf_a_04, bone2Name=dyn_scarf_a_03, swing1LimitDegrees=35, swing2LimitDegrees=45, twistLimitDegrees=3}
 {bone1Name=dyn_scarf_b_01, bone2Name=dyn_scarf_MSTR, swing1LimitDegrees=15, swing2LimitDegrees=35, twistLimitDegrees=2}
 {bone1Name=dyn_scarf_b_02, bone2Name=dyn_scarf_b_01, swing1LimitDegrees=35, swing2LimitDegrees=45, twistLimitDegrees=2}
 {bone1Name=dyn_scarf_b_03, bone2Name=dyn_scarf_b_02, swing1LimitDegrees=35, swing2LimitDegrees=45, twistLimitDegrees=3}
 {
  bone1Name=dyn_scarf_a_02
  bone2Name=dyn_scarf_b_02
  swing1LimitDegrees=45
  swing1Motion=Free
  swing2LimitDegrees=45
  swing2Motion=Free
  twistLimitDegrees=45
  twistMotion=Free
 }
 {
  bone1Name=dyn_quiver
  bone2Name=pelvis
  swing1LimitDegrees=45
  swing1Motion=Free
  swing2LimitDegrees=45
  swing2Motion=Free
  twistLimitDegrees=45
  twistMotion=Free
 }
 {
  bone1Name=dyn_skirtFront_01_l
  bone2Name=pelvis
  swing1LimitDegrees=45
  swing1Motion=Free
  swing2LimitDegrees=45
  swing2Motion=Free
  twistLimitDegrees=45
  twistMotion=Free
 }
 {
  bone1Name=dyn_skirtFront_02_l
  bone2Name=dyn_skirtFront_01_l
  swing1LimitDegrees=45
  swing1Motion=Free
  swing2LimitDegrees=45
  swing2Motion=Free
  twistLimitDegrees=45
  twistMotion=Free
 }
 {
  bone1Name=dyn_scarf_a_03
  bone2Name=dyn_scarf_b_03
  swing1LimitDegrees=45
  swing1Motion=Free
  swing2LimitDegrees=45
  swing2Motion=Free
  twistLimitDegrees=45
  twistMotion=Free
 }
]
