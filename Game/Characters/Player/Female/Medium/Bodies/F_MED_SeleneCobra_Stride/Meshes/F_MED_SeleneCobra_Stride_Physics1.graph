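# Physics asset for the skirt of the female medium SeleneCobra Stride player body.
# Back, side and front skirt bone chains use capsules and angular-limited joints, hanging from tiny master bodies with free rotation on the pelvis.
# Pelvis, thigh, calf and right-foot capsules form the lower body, with no left foot or front-right chain.
# An unconstrained Floor_COL body adds a huge offset sphere, apparently approximating ground, plus a smaller one.
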PhysicsAsset /Game/Characters/Player/Female/Medium/Bodies/F_MED_SeleneCobra_Stride/Meshes/F_MED_SeleneCobra_Stride_Physics1
bodies {
 Floor_COL [
  {shapeType=Sphere, radius=212, shapeName=floorCOL_sphere, center=<-3.4923 -2.714 -209>}
  {shapeType=Sphere, radius=27.2085, shapeName=floorCOL_sphere, center=<0 -2.2038 -0.4285>}
 ]
 calf_l [{shapeType=Capsule, length=40.1602, radius=7, center=<-24.0798 -2.1574 0.048>, rotation=<rot 0 89.5329 -88.5703>}]
 calf_r [{shapeType=Capsule, length=46.6602, radius=7.5, center=<26.1858 1.9985 0.005>, rotation=<rot -0.3385 89.5329 -88.4634>}]
 dyn_skirtBack_01 [{shapeType=Capsule, length=14.875, radius=6.125, shapeName=dyn_skirtBack_01_l_capsule, center=<5 0 5>, rotation=<rot 0 0 90>}]
 dyn_skirtBack_01_l [{shapeType=Capsule, length=15.375, radius=5.625, shapeName=dyn_skirtBack_01_l_capsule, center=<5 0 6>, rotation=<rot -0.0001 0.0001 84.9997>}]
 dyn_skirtBack_01_r [{shapeType=Capsule, length=15.375, radius=5.625, shapeName=dyn_skirtBack_01_r_capsule, center=<5 0 6>, rotation=<rot -0.0004 -0.0004 95>}]
 dyn_skirtBack_02 [{shapeType=Capsule, length=18.375, radius=7.375, shapeName=dyn_skirtBack_02_l_capsule, center=<5 0 5>, rotation=<rot 0 0 90>}]
 dyn_skirtBack_02_l [{shapeType=Capsule, length=18.375, radius=7.375, shapeName=dyn_skirtBack_02_l_capsule, center=<5 3 6>, rotation=<rot 0 0 94.9998>}]
 dyn_skirtBack_02_r [{shapeType=Capsule, length=17.125, radius=8.625, shapeName=dyn_skirtBack_02_r_capsule, center=<5 -3 6>, rotation=<rot 0 0 89.9986>}]
 dyn_skirtBack_03 [{shapeType=Capsule, length=25.3243, radius=8.75, shapeName=dyn_skirtBack_03_l_capsule, center=<5 0 5>, rotation=<rot 0 0 90>}]
 dyn_skirtBack_03_l [{shapeType=Capsule, length=25.3243, radius=8.75, shapeName=dyn_skirtBack_03_l_capsule, center=<5 3 6>, rotation=<rot 0 0 94.9999>}]
 dyn_skirtBack_03_r [{shapeType=Capsule, length=24.1993, radius=9.875, shapeName=dyn_skirtBack_03_r_capsule, center=<5 -3 6>, rotation=<rot 0 0 89.9986>}]
 dyn_skirtBack_04_r [{shapeType=Capsule, length=25.3243, radius=10.75, shapeName=dyn_skirtBack_04_r_capsule, center=<5 -3 6>, rotation=<rot 0 0 89.999>}]
 dyn_skirtBack_mstr [{shapeType=Capsule, length=0.505, radius=0.505, shapeName=dyn_jacketSide_mstr_r_capsule, rotation=<rot 0 -90 0>}]
 dyn_skirtBack_mstr_l [{shapeType=Capsule, length=0.505, radius=0.505, shapeName=dyn_jacketSide_mstr_r_capsule, rotation=<rot 0 -90 0>}]
 dyn_skirtBack_mstr_r [{shapeType=Capsule, length=0.505, radius=0.505, shapeName=dyn_jacketSide_mstr_r_capsule, rotation=<rot 0 -90 0>}]
 dyn_skirtFront_01 [{shapeType=Capsule, length=0.505, radius=0.505, shapeName=dyn_skirtFront_01_capsule, rotation=<rot 0 -90 0>}]
 dyn_skirtFront_01_l [{
   shapeType=Capsule
   length=1.8982
   radius=1.6067
   shapeName=dyn_skirtFront_01_l_capsule
   center=<7.3583 0.5446 0.0241>
   rotation=<rot 0 -38.4311 -85.8602>
  }]
 dyn_skirtFront_02 [{
   shapeType=Capsule
   length=4.5179
   radius=3.835
   shapeName=dyn_skirtFront_02_capsule
   center=<6.664 -0.6749 0.033>
   rotation=<rot 0 -16.6185 -81.69>
  }]
 dyn_skirtFront_02_l [{shapeType=Capsule, length=5.841, radius=7.625, shapeName=dyn_skirtSide_03_l_capsule, center=<6 0 5>, rotation=<rot 0.0003 0.0003 100.8886>}]
 dyn_skirtFront_03 [{shapeType=Capsule, length=4.091, radius=6.375, shapeName=dyn_skirtSide_04_l_capsule, center=<7 -3 5>, rotation=<rot 90 0 0>}]
 dyn_skirtFront_03_l [{shapeType=Capsule, length=10.091, radius=6.875, shapeName=dyn_skirtSide_03_l_capsule, center=<6 0 5>, rotation=<rot 0.0002 0.0002 100.8886>}]
 dyn_skirtFront_04 [{shapeType=Capsule, length=4.091, radius=6.375, shapeName=dyn_skirtSide_04_l_capsule, center=<7 -3 5>, rotation=<rot 90 0 0>}]
 dyn_skirtFront_04_l [{shapeType=Capsule, length=7.966, radius=6.5, shapeName=dyn_skirtSide_04_l_capsule, center=<6 0 5>, rotation=<rot 0.0001 0.0002 100.8885>}]
 dyn_skirtFront_05 [{shapeType=Capsule, length=1.6369, radius=6.3228, shapeName=dyn_skirtFront_05_capsule, center=<4 -3 5>, rotation=<rot 90 0 0>}]
 dyn_skirtFront_mstr [{shapeType=Capsule, length=0.505, radius=0.505, shapeName=dyn_jacketSide_mstr_r_capsule, rotation=<rot 0 -90 0>}]
 dyn_skirtFront_mstr_l [{shapeType=Capsule, length=0.505, radius=0.505, shapeName=dyn_jacketSide_mstr_r_capsule, rotation=<rot 0 -90 0>}]
 dyn_skirtSide_01_l [{shapeType=Capsule, length=0.505, radius=0.505, shapeName=dyn_skirtFront_01_l_capsule, rotation=<rot 0 -90 0>}]
 dyn_skirtSide_01_r [{shapeType=Capsule, length=17.375, radius=6.875, shapeName=dyn_skirtSide_01_r_capsule, center=<6 0 6>, rotation=<rot 0 0 90>}]
 dyn_skirtSide_02_l [{shapeType=Capsule, length=21.2769, radius=8, shapeName=dyn_skirtSide_02_l_capsule, center=<5 0 5>, rotation=<rot 0 0.0001 80.8878>}]
 dyn_skirtSide_02_r [{shapeType=Capsule, length=17.125, radius=7.125, shapeName=dyn_skirtSide_02_r_capsule, center=<6 0 6>, rotation=<rot 0 0 90>}]
 dyn_skirtSide_03_l [{shapeType=Capsule, length=27.591, radius=7.625, shapeName=dyn_skirtSide_03_l_capsule, center=<5 0 5>, rotation=<rot 0 0.0001 80.8887>}]
 dyn_skirtSide_03_r [{shapeType=Capsule, length=16.875, radius=8.375, shapeName=dyn_skirtSide_03_r_capsule, center=<6 0 6>, rotation=<rot 0 0 90>}]
 dyn_skirtSide_04_l [{shapeType=Capsule, length=27.966, radius=8.25, shapeName=dyn_skirtSide_04_l_capsule, center=<5 0 5>, rotation=<rot 0 0.0001 80.8885>}]
 dyn_skirtSide_04_r [{shapeType=Capsule, length=16.75, radius=9.5, shapeName=dyn_skirtSide_04_r_capsule, center=<6 0 6>, rotation=<rot 0 0 90>}]
 dyn_skirtSide_05_r [
  {shapeType=Capsule, length=16.75, radius=9.5, shapeName=dyn_skirtSide_04_r_capsule, center=<6 0 6>, rotation=<rot -0.0001 -0.0004 85.0001>}
  {
   shapeType=Capsule
   length=9.875
   radius=3.375
   shapeName=dyn_skirtSide_05_r_capsule
   center=<9.2017 -10.3089 -2.9958>
   rotation=<rot -82.1768 -178.5989 173.6149>
  }
 ]
 dyn_skirtSide_mstr_l1 [{shapeType=Capsule, length=0.505, radius=0.505, shapeName=dyn_jacketSide_mstr_r_capsule, rotation=<rot 0 -90 0>}]
 dyn_skirtSide_mstr_r [{shapeType=Capsule, length=0.505, radius=0.505, shapeName=dyn_jacketSide_mstr_r_capsule, rotation=<rot 0 -90 0>}]
 foot_r [{
   shapeType=Capsule
   length=13.3752
   radius=10.2462
   shapeName=foot_r_capsule
   center=<1.861 -5.5637 -3.4042>
   rotation=<rot 1.2717 5.0624 -87.7965>
  }]
 pelvis [
  {shapeType=Capsule, length=7.7125, radius=8.4301, center=<-0.9227 -1.2812 0>}
  {shapeType=Capsule, length=10.79, radius=7.4702, shapeName=pelvis_capsule, center=<-3.9978 -3.5061 0>}
 ]
 thigh_l [{shapeType=Capsule, length=28.625, radius=7.375, center=<-21.6167 -1.0761 1.3426>, rotation=<rot 85 179.8044 179.9999>}]
 thigh_r [{shapeType=Capsule, length=28.125, radius=7.75, center=<21.0533 0.5413 -1.1849>, rotation=<rot 85 -180 -180>}]
}
constraints [
 {bone1Name=thigh_l, bone2Name=pelvis, swing1LimitDegrees=45, swing2LimitDegrees=45, twistLimitDegrees=45}
 {bone1Name=calf_l, bone2Name=thigh_l, swing1LimitDegrees=45, swing2LimitDegrees=45, twistLimitDegrees=45}
 {bone1Name=thigh_r, bone2Name=pelvis, swing1LimitDegrees=45, swing2LimitDegrees=45, twistLimitDegrees=45}
 {bone1Name=calf_r, bone2Name=thigh_r, swing1LimitDegrees=45, swing2LimitDegrees=45, twistLimitDegrees=45}
 {bone1Name=dyn_skirtSide_05_r, bone2Name=dyn_skirtSide_04_r, swing1LimitDegrees=15, swing2LimitDegrees=35, twistLimitDegrees=15}
 {
  bone1Name=dyn_skirtBack_mstr
  bone2Name=pelvis
  swing1LimitDegrees=45
  swing1Motion=Free
  swing2LimitDegrees=45
  swing2Motion=Free
  twistLimitDegrees=45
  twistMotion=Free
 }
 {bone1Name=dyn_skirtBack_01, bone2Name=dyn_skirtBack_mstr, swing1LimitDegrees=15, swing2LimitDegrees=35, twistLimitDegrees=5}
 {bone1Name=dyn_skirtBack_02, bone2Name=dyn_skirtBack_01, swing1LimitDegrees=15, swing2LimitDegrees=35, twistLimitDegrees=10}
 {bone1Name=dyn_skirtBack_03, bone2Name=dyn_skirtBack_02, swing1LimitDegrees=15, swing2LimitDegrees=35, twistLimitDegrees=10}
 {
  bone1Name=dyn_skirtBack_mstr_l
  bone2Name=pelvis
  swing1LimitDegrees=45
  swing1Motion=Free
  swing2LimitDegrees=45
  swing2Motion=Free
  twistLimitDegrees=45
  twistMotion=Free
 }
 {bone1Name=dyn_skirtBack_01_l, bone2Name=dyn_skirtBack_mstr_l, swing1LimitDegrees=15, swing2LimitDegrees=35, twistLimitDegrees=15}
 {bone1Name=dyn_skirtBack_02_l, bone2Name=dyn_skirtBack_01_l, swing1LimitDegrees=15, swing2LimitDegrees=35, twistLimitDegrees=15}
 {bone1Name=dyn_skirtBack_03_l, bone2Name=dyn_skirtBack_02_l, swing1LimitDegrees=15, swing2LimitDegrees=35, twistLimitDegrees=15}
 {
  bone1Name=dyn_skirtBack_mstr_r
  bone2Name=pelvis
  swing1LimitDegrees=45
  swing1Motion=Free
  swing2LimitDegrees=45
  swing2Motion=Free
  twistLimitDegrees=45
  twistMotion=Free
 }
 {bone1Name=dyn_skirtBack_01_r, bone2Name=dyn_skirtBack_mstr_r, swing1LimitDegrees=15, swing2LimitDegrees=35, twistLimitDegrees=15}
 {bone1Name=dyn_skirtBack_02_r, bone2Name=dyn_skirtBack_01_r, swing1LimitDegrees=15, swing2LimitDegrees=35, twistLimitDegrees=15}
 {bone1Name=dyn_skirtBack_03_r, bone2Name=dyn_skirtBack_02_r, swing1LimitDegrees=15, swing2LimitDegrees=35, twistLimitDegrees=15}
 {bone1Name=dyn_skirtBack_04_r, bone2Name=dyn_skirtBack_03_r, swing1LimitDegrees=15, swing2LimitDegrees=35, twistLimitDegrees=15}
 {
  bone1Name=dyn_skirtFront_mstr
  bone2Name=pelvis
  swing1LimitDegrees=45
  swing1Motion=Free
  swing2LimitDegrees=45
  swing2Motion=Free
  twistLimitDegrees=45
  twistMotion=Free
 }
 {
  bone1Name=dyn_skirtFront_mstr_l
  bone2Name=pelvis
  swing1LimitDegrees=45
  swing1Motion=Free
  swing2LimitDegrees=45
  swing2Motion=Free
  twistLimitDegrees=45
  twistMotion=Free
 }
 {
  bone1Name=dyn_skirtSide_mstr_l1
  bone2Name=pelvis
  swing1LimitDegrees=45
  swing1Motion=Free
  swing2LimitDegrees=45
  swing2Motion=Free
  twistLimitDegrees=45
  twistMotion=Free
 }
 {
  bone1Name=dyn_skirtSide_01_l
  bone2Name=dyn_skirtSide_mstr_l1
  swing1LimitDegrees=45
  swing1Motion=Free
  swing2LimitDegrees=45
  swing2Motion=Free
  twistLimitDegrees=45
  twistMotion=Free
 }
 {bone1Name=dyn_skirtSide_02_l, bone2Name=dyn_skirtSide_01_l, swing1LimitDegrees=35, swing2LimitDegrees=35, twistLimitDegrees=15}
 {bone1Name=dyn_skirtSide_03_l, bone2Name=dyn_skirtSide_02_l, swing1LimitDegrees=35, swing2LimitDegrees=35, twistLimitDegrees=15}
 {bone1Name=dyn_skirtSide_04_l, bone2Name=dyn_skirtSide_03_l, swing1LimitDegrees=35, swing2LimitDegrees=35, twistLimitDegrees=15}
 {
  bone1Name=dyn_skirtSide_mstr_r
  bone2Name=pelvis
  swing1LimitDegrees=45
  swing1Motion=Free
  swing2LimitDegrees=45
  swing2Motion=Free
  twistLimitDegrees=45
  twistMotion=Free
 }
 {bone1Name=dyn_skirtSide_01_r, bone2Name=dyn_skirtSide_mstr_r, swing1LimitDegrees=15, swing2LimitDegrees=35, twistLimitDegrees=15}
 {bone1Name=dyn_skirtSide_02_r, bone2Name=dyn_skirtSide_01_r, swing1LimitDegrees=15, swing2LimitDegrees=35, twistLimitDegrees=15}
 {bone1Name=dyn_skirtSide_03_r, bone2Name=dyn_skirtSide_02_r, swing1LimitDegrees=15, swing2LimitDegrees=35, twistLimitDegrees=15}
 {bone1Name=dyn_skirtSide_04_r, bone2Name=dyn_skirtSide_03_r, swing1LimitDegrees=15, swing2LimitDegrees=35, twistLimitDegrees=15}
 {bone1Name=dyn_skirtFront_04_l, bone2Name=dyn_skirtFront_03_l, swing1LimitDegrees=35, swing2LimitDegrees=35, twistLimitDegrees=15}
 {bone1Name=dyn_skirtFront_03_l, bone2Name=dyn_skirtFront_02_l, swing1LimitDegrees=35, swing2LimitDegrees=35, twistLimitDegrees=15}
 {bone1Name=dyn_skirtFront_03, bone2Name=dyn_skirtFront_02, swing1LimitDegrees=25, swing2LimitDegrees=35, twistLimitDegrees=15}
 {
  bone1Name=dyn_skirtFront_02
  bone2Name=dyn_skirtFront_01
  swing1LimitDegrees=45
  swing1Motion=Free
  swing2LimitDegrees=45
  swing2Motion=Free
  twistLimitDegrees=45
  twistMotion=Free
 }
 {
  bone1Name=dyn_skirtFront_01
  bone2Name=dyn_skirtFront_mstr
  swing1LimitDegrees=45
  swing1Motion=Free
  swing2LimitDegrees=45
  swing2Motion=Free
  twistLimitDegrees=45
  twistMotion=Free
 }
 {
  bone1Name=dyn_skirtFront_02_l
  bone2Name=dyn_skirtFront_01_l
  swing1LimitDegrees=45
  swing1Motion=Free
  swing2LimitDegrees=45
  swing2Motion=Free
  twistLimitDegrees=45
  twistMotion=Free
 }
 {
  bone1Name=dyn_skirtFront_01_l
  bone2Name=dyn_skirtFront_mstr_l
  swing1LimitDegrees=45
  swing1Motion=Free
  swing2LimitDegrees=45
  swing2Motion=Free
  twistLimitDegrees=45
  twistMotion=Free
 }
 {bone1Name=dyn_skirtFront_05, bone2Name=dyn_skirtFront_04, swing1LimitDegrees=35, swing2LimitDegrees=35, twistLimitDegrees=15}
 {bone1Name=dyn_skirtFront_04, bone2Name=dyn_skirtFront_03, swing1LimitDegrees=25, swing2LimitDegrees=35, twistLimitDegrees=15}
 {
  bone1Name=foot_r
  bone2Name=calf_r
  swing1LimitDegrees=45
  swing1Motion=Free
  swing2LimitDegrees=45
  swing2Motion=Free
  twistLimitDegrees=45
  twistMotion=Free
 }
]
