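# Animation Blueprint for the female Selene Cobra Stride player body, adding secondary motion to scarf, skirt, chains, quiver and charms over the attached parent's copied pose.
# Bone constraints, a Control Rig and slide offsets build a cached pre-simulation pose.
# Two RigidBody branches blend on a backpack flag, differing in whether scarf-to-spine constraints precede or follow simulation.
# Movement-state bone offsets and AnimDynamics follow; a custom-idle slot layers over the quiver.
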AnimBlueprint /Game/Characters/Player/Female/Medium/Bodies/F_MED_SeleneCobra_Stride/Meshes/F_MED_SeleneCobra_Stride_AnimBP
target_skeleton=&/F_MED_SeleneCobra_Stride_Skeleton
graphs [
 AnimGraph {nodes [
   n1 = "Transform (Modify) Bone - Bone: dyn_scarf_b_04"!(ComponentPose, bAlphaBoolEnabled:bool=True, Translation="0,0,0", Rotation="5,0,0", Scale="1,1,1") -> (Pose) {
    alphaInputType=Bool
    rotationMode=BMM_Additive
    rotationSpace=BCS_BoneSpace
    alphaBoolBlend {blendInTime=0.1, blendOption=Linear, blendOutTime=0.1}
    boneToModify {boneName=dyn_scarf_b_04}
    rotation {pitch=5}
   }
   "Transform (Modify) Bone - Bone: dyn_scarf_a_04"!(ComponentPose=@n1, bAlphaBoolEnabled:bool=True, Translation="0,0,0", Rotation="5,0,0", Scale="1,1,1") -> (Pose) {
    alphaInputType=Bool
    rotationMode=BMM_Additive
    rotationSpace=BCS_BoneSpace
    alphaBoolBlend {blendInTime=0.1, blendOption=Linear, blendOutTime=0.1}
    boneToModify {boneName=dyn_scarf_a_04}
    rotation {pitch=5}
   }
   AnimGraphNode_Slot_customIdle_0 = Slot 'DefaultSlot'()
   AnimGraphNode_UseCachedPose_customIdle_0 = Use cached pose 'startpose'()
   K2Node_VariableGet_customIdle_1 = Get Custom Idle is Playing() -> (Custom Idle is Playing:float)
   AnimGraphNode_LayeredBoneBlend_customIdle_0 = Layered blend per bone(
    BasePose=@AnimGraphNode_UseCachedPose_customIdle_0
    BlendPoses_0=@AnimGraphNode_Slot_customIdle_0
    BlendWeights_0=@K2Node_VariableGet_customIdle_1
   ) {layerSetup [{branchFilters [{boneName=dyn_Quiver_MSTR}]}]}
   Output Pose(Result=@AnimGraphNode_LayeredBoneBlend_customIdle_0) {layerGroup=None}
   n2 = Look At - Bone: dyn_scarf_MSTR(ComponentPose=Local To Component(LocalPose=Copy Pose From Mesh() {bUseAttachedParent=true}), Alpha=1) {
    alphaBoolBlend {blendOption=Linear}
    boneToModify {boneName=dyn_scarf_MSTR}
    lookAtLocation {x=0}
    lookAtTarget {boneReference {boneName=pelvis}}
    lookAt_Axis {axis=<1 0>}
   }
   n3 = "Constraint - dyn_chain_frt_MSTR (PPPPPPPPP)"!(
    ComponentPose=@n2
    Alpha=1
    ConstraintWeights_0=0.4894
    ConstraintWeights_1:float=0.2026
    ConstraintWeights_2:float=0.0398
    ConstraintWeights_3:float=0.1841
    ConstraintWeights_4:float=0.0089
    ConstraintWeights_5:float=0.0037
    ConstraintWeights_6:float=0.0474
    ConstraintWeights_7:float=0.0081
    ConstraintWeights_8:float=0.0159
   ) -> (Pose) {
    alphaBoolBlend {blendOption=Linear}
    boneToModify {boneName=dyn_chain_frt_MSTR}
    constraintSetup [
     {targetBone {boneName=pelvis}}
     {targetBone {boneName=thigh_twist_01_l}}
     {targetBone {boneName=thigh_twist_01_r}}
     {targetBone {boneName=deform_groin_l}}
     {targetBone {boneName=deform_groin_r}}
     {targetBone {boneName=dyn_skirtFront_02_l}}
     {targetBone {boneName=dyn_skirtFront_01_l}}
     {targetBone {boneName=dyn_skirtFront_01}}
     {targetBone {boneName=dyn_skirtSide_01_l}}
    ]
    constraintWeights [1,1,1,1,1,1,1,1,1]
   }
   n4 = "Constraint - dyn_chain_bk_MSTR (PPPPPP)"!(
    ComponentPose=@n3
    Alpha=1
    ConstraintWeights_0=0.0171
    ConstraintWeights_1:float=0.0194
    ConstraintWeights_2:float=0.1937
    ConstraintWeights_3:float=0.0856
    ConstraintWeights_4:float=0.6842
    ConstraintWeights_5:float=0
   ) -> (Pose) {
    alphaBoolBlend {blendOption=Linear}
    boneToModify {boneName=dyn_chain_bk_MSTR}
    constraintSetup [
     {targetBone {boneName=spine_04}}
     {targetBone {boneName=spine_03}}
     {targetBone {boneName=spine_02}}
     {targetBone {boneName=spine_01}}
     {targetBone {boneName=pelvis}}
     {targetBone {boneName=deform_glute_l}}
    ]
    constraintWeights [1,1,1,1,1,1]
   }
   n5 = "Constraint - dyn_scarf_MSTR (PPPPPPPP)"!(
    ComponentPose=@n4
    Alpha=1
    ConstraintWeights_0=0.3258
    ConstraintWeights_1:float=0.6203
    ConstraintWeights_2:float=0.0088
    ConstraintWeights_3:float=0.0202
    ConstraintWeights_4:float=0.0098
    ConstraintWeights_5:float=0.0003
    ConstraintWeights_6:float=0.0076
    ConstraintWeights_7:float=0.0071
   ) -> (Pose) {
    alphaBoolBlend {blendOption=Linear}
    boneToModify {boneName=dyn_scarf_MSTR}
    constraintSetup [
     {targetBone {boneName=spine_05}}
     {targetBone {boneName=clavicle_l}}
     {targetBone {boneName=clavicle_r}}
     {targetBone {boneName=spine_04}}
     {targetBone {boneName=upperarm_twist_01_l}}
     {targetBone {boneName=spine_03}}
     {targetBone {boneName=deform_upperarm_bk_l}}
     {targetBone {boneName=deform_upperarm_up_l}}
    ]
    constraintWeights [1,1,1,1,1,1,1,1]
   }
   n6 = "Transform (Modify) Bone - Bone: Floor_COL"(ComponentPose=@n5, Alpha=Get jumping_chain() -> (jumping_chain:double), Translation="0,0,-100", Rotation="0,0,0", Scale="1,1,1") {
    rotationMode=BMM_Additive
    rotationSpace=BCS_BoneSpace
    translationMode=BMM_Additive
    alphaBoolBlend {blendOption=Linear}
    alphaScaleBiasClamp {bInterpResult=true}
    boneToModify {boneName=Floor_COL}
    translation {z=-100}
   }
   n7 = Control Rig!(
    Source=Component To Local(ComponentPose=@n6)
    Alpha
    jumping:double=Get jumping() -> (jumping:double)
    Moving2d:bool=Get bIsMoving2D()
    jumping_chain:double=Get jumping_chain()
    FrontendIdle:double=Get FrontendIdle() -> (FrontendIdle:float)
   ) -> (Pose) {
    alphaBoolBlend {blendOption=Linear}
    controlRigAssetReference {blueprintRigClass=&/F_MED_SeleneCobra_Stride_CtrlRig.F_MED_SeleneCobra_Stride_CtrlRig_C}
   }
   n8 = Get bIsSliding()
   n9 = "Transform (Modify) Bone - Bone: dyn_skirtBack_mstr"!(ComponentPose=Local To Component(LocalPose=@n7), bAlphaBoolEnabled:bool=@n8, Translation="0,0,0", Rotation="90,0,0", Scale="1,1,1") -> (Pose) {
    alphaInputType=Bool
    rotationMode=BMM_Additive
    rotationSpace=BCS_BoneSpace
    alphaBoolBlend {blendInTime=0.1, blendOption=Linear, blendOutTime=0.1}
    boneToModify {boneName=dyn_skirtBack_mstr}
    rotation {pitch=90}
   }
   n10 = "Transform (Modify) Bone - Bone: dyn_skirtBack_mstr_l"!(ComponentPose=@n9, bAlphaBoolEnabled:bool=@n8, Translation="0,0,0", Rotation="70,0,0", Scale="1,1,1") -> (Pose) {
    alphaInputType=Bool
    rotationMode=BMM_Additive
    rotationSpace=BCS_BoneSpace
    alphaBoolBlend {blendInTime=0.1, blendOption=Linear, blendOutTime=0.1}
    boneToModify {boneName=dyn_skirtBack_mstr_l}
    rotation {pitch=70}
   }
   n11 = "Transform (Modify) Bone - Bone: dyn_skirtBack_mstr_r"!(ComponentPose=@n10, bAlphaBoolEnabled:bool=@n8, Translation="0,0,0", Rotation="70,0,0", Scale="1,1,1") -> (Pose) {
    alphaInputType=Bool
    rotationMode=BMM_Additive
    rotationSpace=BCS_BoneSpace
    alphaBoolBlend {blendInTime=0.1, blendOption=Linear, blendOutTime=0.1}
    boneToModify {boneName=dyn_skirtBack_mstr_r}
    rotation {pitch=70}
   }
   n12 = "Transform (Modify) Bone - Bone: dyn_skirtSide_mstr_l1"!(ComponentPose=@n11, bAlphaBoolEnabled:bool=@n8, Translation="0,0,0", Rotation="45,0,0", Scale="1,1,1") -> (Pose) {
    alphaInputType=Bool
    rotationMode=BMM_Additive
    rotationSpace=BCS_BoneSpace
    alphaBoolBlend {blendInTime=0.1, blendOption=Linear, blendOutTime=0.1}
    boneToModify {boneName=dyn_skirtSide_mstr_l1}
    rotation {pitch=45}
   }
   n13 = "Transform (Modify) Bone - Bone: dyn_skirtSide_mstr_r"!(ComponentPose=@n12, bAlphaBoolEnabled:bool=@n8, Translation="0,0,0", Rotation="45,0,0", Scale="1,1,1") -> (Pose) {
    alphaInputType=Bool
    rotationMode=BMM_Additive
    rotationSpace=BCS_BoneSpace
    alphaBoolBlend {blendInTime=0.1, blendOption=Linear, blendOutTime=0.1}
    boneToModify {boneName=dyn_skirtSide_mstr_r}
    rotation {pitch=45}
   }
   n14 = "Transform (Modify) Bone - Bone: dyn_Quiver_MSTR"!(ComponentPose=@n13, bAlphaBoolEnabled:bool=@n8, Translation="-1,0,5", Rotation="35,30,-15", Scale="1,1,1") -> (Pose) {
    alphaInputType=Bool
    rotationMode=BMM_Additive
    rotationSpace=BCS_BoneSpace
    translationMode=BMM_Additive
    translationSpace=BCS_BoneSpace
    alphaBoolBlend {blendInTime=0.1, blendOption=Linear, blendOutTime=0.1}
    boneToModify {boneName=dyn_Quiver_MSTR}
    rotation=<rot 35 30 -15>
    translation {x=-1, z=5}
   }
   n15 = "Transform (Modify) Bone - Bone: dyn_Quiver_MSTR"(ComponentPose=@n14, bAlphaBoolEnabled=Get bIsBaseballSliding(), Translation="0,0,0", Rotation="0.007,0,7", Scale="1,1,1") {
    alphaInputType=Bool
    rotationMode=BMM_Additive
    rotationSpace=BCS_BoneSpace
    translationMode=BMM_Additive
    translationSpace=BCS_BoneSpace
    alphaBoolBlend {blendInTime=0.1, blendOption=Linear, blendOutTime=0.1}
    boneToModify {boneName=dyn_Quiver_MSTR}
    rotation {pitch=0.007, roll=7}
   }
   Save cached pose 'presim'(Pose=Component To Local(ComponentPose=@n15)) {cacheName=presim}
   n16 = "Constraint - dyn_scarf_b_01 (P)"(ComponentPose=Local To Component(LocalPose=Use cached pose 'presim'()), Alpha=0.8, ConstraintWeights_0=1) {
    alphaBoolBlend {blendOption=Linear}
    boneToModify {boneName=dyn_scarf_b_01}
    constraintSetup [{targetBone {boneName=spine_05}}]
    constraintWeights [1]
   }
   n17 = "Constraint - dyn_scarf_a_01 (P)"(ComponentPose=@n16, Alpha=0.8, ConstraintWeights_0=1) {
    alphaBoolBlend {blendOption=Linear}
    boneToModify {boneName=dyn_scarf_a_01}
    constraintSetup [{targetBone {boneName=spine_05}}]
    constraintWeights [1]
   }
   n18 = "Constraint - dyn_scarf_b_02 (PP)"!(ComponentPose=@n17, Alpha=0.7, ConstraintWeights_0=0.7, ConstraintWeights_1:float=0.3) -> (Pose) {
    alphaBoolBlend {blendOption=Linear}
    boneToModify {boneName=dyn_scarf_b_02}
    constraintSetup [{targetBone {boneName=spine_04}}, {targetBone {boneName=spine_03}}]
    constraintWeights [1,1]
   }
   n19 = "Constraint - dyn_scarf_a_02 (PP)"!(ComponentPose=@n18, Alpha=0.7, ConstraintWeights_0=0.7, ConstraintWeights_1:float=0.3) -> (Pose) {
    alphaBoolBlend {blendOption=Linear}
    boneToModify {boneName=dyn_scarf_a_02}
    constraintSetup [{targetBone {boneName=spine_04}}, {targetBone {boneName=spine_03}}]
    constraintWeights [1,1]
   }
   n20 = "Constraint - dyn_scarf_b_03 (P)"(ComponentPose=@n19, Alpha=0.5, ConstraintWeights_0=1) {
    alphaBoolBlend {blendOption=Linear}
    boneToModify {boneName=dyn_scarf_b_03}
    constraintSetup [{targetBone {boneName=spine_02}}]
    constraintWeights [1]
   }
   n21 = "Constraint - dyn_scarf_a_03 (P)"(ComponentPose=@n20, Alpha=0.5, ConstraintWeights_0=1) {
    alphaBoolBlend {blendOption=Linear}
    boneToModify {boneName=dyn_scarf_a_03}
    constraintSetup [{targetBone {boneName=spine_02}}]
    constraintWeights [1]
   }
   n22 = "Constraint - dyn_scarf_b_04 (P)"(ComponentPose=@n21, Alpha=1, ConstraintWeights_0=1) {
    alphaBoolBlend {blendOption=Linear}
    boneToModify {boneName=dyn_scarf_b_04}
    constraintSetup [{targetBone {boneName=pelvis}}]
    constraintWeights [1]
   }
   n23 = "Constraint - dyn_scarf_a_04 (P)"(ComponentPose=@n22, Alpha=1, ConstraintWeights_0=1) {
    alphaBoolBlend {blendOption=Linear}
    boneToModify {boneName=dyn_scarf_a_04}
    constraintSetup [{targetBone {boneName=pelvis}}]
    constraintWeights [1]
   }
   n24 = RigidBody!(ComponentPose=@n23, bAlphaBoolEnabled:bool, OverrideWorldGravity:Vector=Get GravityOverride(), ExternalForce) -> (Pose) {
    alphaInputType=Bool
    bClampLinearTranslationLimitToRefPose=true
    bOverrideWorldGravity=true
    overridePhysicsAsset=&/F_MED_SeleneCobra_Stride_Physics1
    simulationSpace=BaseBoneSpace
    alphaBoolBlend {blendInTime=0.1, blendOption=Linear, blendOutTime=0.1}
    baseBoneRef {boneName=pelvis}
    componentAppliedLinearAccClamp=<3000 3000 3000>
    componentLinearAccScale=<0.3 0.3 0.5>
    componentLinearVelScale=<0.3 0.3 0.5>
    simSpaceSettings {
     maxAngularAcceleration=1000
     maxAngularVelocity=3
     maxLinearAcceleration=700
     maxLinearVelocity=500
     velocityScaleZ=0
     worldAlpha=1
     externalLinearDragV=<0.2 0.2>
    }
   }
   n25 = RigidBody!(
    ComponentPose=@n24
    bAlphaBoolEnabled
    OverrideWorldGravity=Get GravityOverride scarf() -> (GravityOverride scarf:Vector)
    ExternalForce
    ComponentLinearAccScale:Vector=Get ACC() -> (ACC:Vector)
    ComponentLinearVelScale:Vector=Get VEL() -> (VEL:Vector)
   ) -> (Pose) {
    alphaInputType=Bool
    bClampLinearTranslationLimitToRefPose=true
    bOverrideWorldGravity=true
    overridePhysicsAsset=&/F_MED_SeleneCobra_Stride_Physics
    simulationSpace=BaseBoneSpace
    alphaBoolBlend {blendInTime=0.1, blendOption=Linear, blendOutTime=0.1}
    baseBoneRef {boneName=spine_02}
    componentAppliedLinearAccClamp=<1500 1500 1500>
    componentLinearAccScale=<0.3 0.3 0.5>
    componentLinearVelScale=<0.3 0.3 0.5>
    simSpaceSettings {
     maxAngularAcceleration=1000
     maxAngularVelocity=3
     maxLinearAcceleration=700
     maxLinearVelocity=500
     velocityScaleZ=0
     externalLinearDragV=<0.2 0.2>
    }
   }
   n26 = Get melee anim() -> (melee anim:double)
   n27 = "Constraint - dyn_scarf_b_01 (PP)"!(ComponentPose=@n25, Alpha=@n26, ConstraintWeights_0=0.3, ConstraintWeights_1:float=0.7) -> (Pose) {
    alphaBoolBlend {blendOption=Linear}
    alphaScaleBiasClamp {bInterpResult=true}
    boneToModify {boneName=dyn_scarf_b_01}
    constraintSetup [{targetBone {boneName=spine_05}}, {targetBone {boneName=clavicle_l}}]
    constraintWeights [1,1]
   }
   n28 = "Constraint - dyn_scarf_a_01 (PP)"!(ComponentPose=@n27, Alpha=@n26, ConstraintWeights_0=0.3, ConstraintWeights_1:float=0.7) -> (Pose) {
    alphaBoolBlend {blendOption=Linear}
    alphaScaleBiasClamp {bInterpResult=true}
    boneToModify {boneName=dyn_scarf_a_01}
    constraintSetup [{targetBone {boneName=spine_05}}, {targetBone {boneName=clavicle_l}}]
    constraintWeights [1,1]
   }
   n29 = Component To Local(ComponentPose=@n28)
   n30 = Get ACC()
   n31 = Get VEL()
   n32 = RigidBody(
    ComponentPose=Local To Component(LocalPose=Use cached pose 'presim'())
    OverrideWorldGravity=Get GravityOverride()
    ComponentLinearAccScale=@n30
    ComponentLinearVelScale=@n31
   ) {
    alphaInputType=Bool
    bClampLinearTranslationLimitToRefPose=true
    bOverrideWorldGravity=true
    overridePhysicsAsset=&F_MED_SeleneCobra_Stride_Physics1
    simulationSpace=BaseBoneSpace
    alphaBoolBlend {blendInTime=0.1, blendOption=Linear, blendOutTime=0.1}
    baseBoneRef {boneName=pelvis}
    componentAppliedLinearAccClamp=<3000 3000 3000>
    componentLinearAccScale=<0.3 0.3 0.5>
    componentLinearVelScale=<0.3 0.3 0.5>
    simSpaceSettings {
     maxAngularAcceleration=1000
     maxAngularVelocity=3
     maxLinearAcceleration=700
     maxLinearVelocity=500
     velocityScaleZ=0
     worldAlpha=1
     externalLinearDragV=<0.2 0.2>
    }
   }
   n33 = RigidBody(ComponentPose=@n32, OverrideWorldGravity=Get GravityOverride scarf(), ComponentLinearAccScale=@n30, ComponentLinearVelScale=@n31) {
    alphaInputType=Bool
    bClampLinearTranslationLimitToRefPose=true
    bOverrideWorldGravity=true
    overridePhysicsAsset=&F_MED_SeleneCobra_Stride_Physics
    simulationSpace=BaseBoneSpace
    alphaBoolBlend {blendInTime=0.1, blendOption=Linear, blendOutTime=0.1}
    baseBoneRef {boneName=pelvis}
    componentAppliedLinearAccClamp=<3000 3000 3000>
    componentLinearAccScale=<0.3 0.3 0.5>
    componentLinearVelScale=<0.3 0.3 0.5>
    simSpaceSettings {
     maxAngularAcceleration=1000
     maxAngularVelocity=3
     maxLinearAcceleration=700
     maxLinearVelocity=500
     velocityScaleZ=0
     externalLinearDragV=<0.2 0.2>
    }
   }
   n34 = "Constraint - dyn_scarf_b_01 (P)"(ComponentPose=@n33, Alpha=0.8, ConstraintWeights_0=1) {
    alphaBoolBlend {blendOption=Linear}
    boneToModify {boneName=dyn_scarf_b_01}
    constraintSetup [{targetBone {boneName=spine_05}}]
    constraintWeights [1]
   }
   n35 = "Constraint - dyn_scarf_a_01 (P)"(ComponentPose=@n34, Alpha=0.8, ConstraintWeights_0=1) {
    alphaBoolBlend {blendOption=Linear}
    boneToModify {boneName=dyn_scarf_a_01}
    constraintSetup [{targetBone {boneName=spine_05}}]
    constraintWeights [1]
   }
   n36 = "Constraint - dyn_scarf_b_02 (PP)"(ComponentPose=@n35, Alpha=0.7, ConstraintWeights_0=0.7, ConstraintWeights_1=0.3) {
    alphaBoolBlend {blendOption=Linear}
    boneToModify {boneName=dyn_scarf_b_02}
    constraintSetup [{targetBone {boneName=spine_04}}, {targetBone {boneName=spine_03}}]
    constraintWeights [1,1]
   }
   n37 = "Constraint - dyn_scarf_a_02 (PP)"(ComponentPose=@n36, Alpha=0.7, ConstraintWeights_0=0.7, ConstraintWeights_1=0.3) {
    alphaBoolBlend {blendOption=Linear}
    boneToModify {boneName=dyn_scarf_a_02}
    constraintSetup [{targetBone {boneName=spine_04}}, {targetBone {boneName=spine_03}}]
    constraintWeights [1,1]
   }
   n38 = "Constraint - dyn_scarf_b_03 (P)"(ComponentPose=@n37, Alpha=0.5, ConstraintWeights_0=1) {
    alphaBoolBlend {blendOption=Linear}
    boneToModify {boneName=dyn_scarf_b_03}
    constraintSetup [{targetBone {boneName=spine_02}}]
    constraintWeights [1]
   }
   n39 = "Constraint - dyn_scarf_a_03 (P)"(ComponentPose=@n38, Alpha=0.5, ConstraintWeights_0=1) {
    alphaBoolBlend {blendOption=Linear}
    boneToModify {boneName=dyn_scarf_a_03}
    constraintSetup [{targetBone {boneName=spine_02}}]
    constraintWeights [1]
   }
   n40 = "Constraint - dyn_scarf_b_04 (P)"(ComponentPose=@n39, Alpha=1, ConstraintWeights_0=1) {
    alphaBoolBlend {blendOption=Linear}
    boneToModify {boneName=dyn_scarf_b_04}
    constraintSetup [{targetBone {boneName=pelvis}}]
    constraintWeights [1]
   }
   n41 = "Constraint - dyn_scarf_a_04 (P)"(ComponentPose=@n40, Alpha=1, ConstraintWeights_0=1) {
    alphaBoolBlend {blendOption=Linear}
    boneToModify {boneName=dyn_scarf_a_04}
    constraintSetup [{targetBone {boneName=pelvis}}]
    constraintWeights [1]
   }
   n42 = "Transform (Modify) Bone - Bone: dyn_scarf_b_01"(ComponentPose=@n41, Alpha=1, Translation="0,0,0", Rotation="-7,0,0", Scale="1,1,1") {
    rotationMode=BMM_Additive
    rotationSpace=BCS_BoneSpace
    alphaBoolBlend {blendOption=Linear}
    boneToModify {boneName=dyn_scarf_b_01}
    rotation {pitch=-7}
   }
   n43 = "Transform (Modify) Bone - Bone: dyn_scarf_a_01"(ComponentPose=@n42, Alpha=1, Translation="0,0,0", Rotation="-7,0,0", Scale="1,1,1") {
    rotationMode=BMM_Additive
    rotationSpace=BCS_BoneSpace
    alphaBoolBlend {blendOption=Linear}
    boneToModify {boneName=dyn_scarf_a_01}
    rotation {pitch=-7}
   }
   n44 = "Transform (Modify) Bone - Bone: dyn_scarf_b_03"(ComponentPose=@n43, Alpha=1, Translation="0,0,0", Rotation="-5,0,0", Scale="1,1,1") {
    rotationMode=BMM_Additive
    rotationSpace=BCS_BoneSpace
    alphaBoolBlend {blendOption=Linear}
    boneToModify {boneName=dyn_scarf_b_03}
    rotation {pitch=-5}
   }
   n45 = "Transform (Modify) Bone - Bone: dyn_scarf_a_03"(ComponentPose=@n44, Alpha=1, Translation="0,0,0", Rotation="-5,0,0", Scale="1,1,1") {
    rotationMode=BMM_Additive
    rotationSpace=BCS_BoneSpace
    alphaBoolBlend {blendOption=Linear}
    boneToModify {boneName=dyn_scarf_a_03}
    rotation {pitch=-5}
   }
   n46 = "Transform (Modify) Bone - Bone: dyn_scarf_b_03"(ComponentPose=@n45, Alpha=1, Translation="0,0,0", Rotation="15,0,0", Scale="1,1,1") {
    rotationMode=BMM_Additive
    rotationSpace=BCS_BoneSpace
    alphaBoolBlend {blendOption=Linear}
    boneToModify {boneName=dyn_scarf_b_03}
    rotation {pitch=15}
   }
   n47 = "Transform (Modify) Bone - Bone: dyn_scarf_a_03"(ComponentPose=@n46, Alpha=1, Translation="0,0,0", Rotation="15,0,0", Scale="1,1,1") {
    rotationMode=BMM_Additive
    rotationSpace=BCS_BoneSpace
    alphaBoolBlend {blendOption=Linear}
    boneToModify {boneName=dyn_scarf_a_03}
    rotation {pitch=15}
   }
   n48 = Get bIsMoving2D()
   n49 = "Transform (Modify) Bone - Bone: dyn_scarf_b_04"(ComponentPose=@n47, bAlphaBoolEnabled=@n48, Translation="0,0,0", Rotation="20,0,0", Scale="1,1,1") {
    alphaInputType=Bool
    rotationMode=BMM_Additive
    rotationSpace=BCS_BoneSpace
    alphaBoolBlend {blendInTime=0.1, blendOption=Linear, blendOutTime=0.1}
    boneToModify {boneName=dyn_scarf_b_04}
    rotation {pitch=20}
   }
   n50 = "Transform (Modify) Bone - Bone: dyn_scarf_a_04"(ComponentPose=@n49, bAlphaBoolEnabled=@n48, Translation="0,0,0", Rotation="20,0,0", Scale="1,1,1") {
    alphaInputType=Bool
    rotationMode=BMM_Additive
    rotationSpace=BCS_BoneSpace
    alphaBoolBlend {blendInTime=0.1, blendOption=Linear, blendOutTime=0.1}
    boneToModify {boneName=dyn_scarf_a_04}
    rotation {pitch=20}
   }
   n51 = "Transform (Modify) Bone - Bone: dyn_scarf_b_02"!(ComponentPose=@n50, bAlphaBoolEnabled:bool=Get bIsInFrontEnd(), Translation="0,0,0", Rotation="5,0,0", Scale="1,1,1") -> (Pose) {
    alphaInputType=Bool
    rotationMode=BMM_Additive
    rotationSpace=BCS_BoneSpace
    alphaBoolBlend {blendInTime=0.1, blendOption=Linear, blendOutTime=0.1}
    boneToModify {boneName=dyn_scarf_b_02}
    rotation {pitch=5}
   }
   n52 = Component To Local(ComponentPose=@n51)
   n53 = Get backpack_bool() -> (backpack_bool:bool)
   n54 = Get crouch still() -> (crouch still:double)
   n55 = "Transform (Modify) Bone - Bone: dyn_skirtFront_03"(
    ComponentPose=Local To Component(LocalPose=Blend Poses by bool(BlendPose_0=@n52, BlendPose_1=@n29, bActiveValue=@n53) {blendType=Linear})
    Alpha=@n54
    Translation="0,-0.5,0"
    Rotation="-5,0,0"
    Scale="1,1,1"
   ) {
    rotationMode=BMM_Additive
    rotationSpace=BCS_BoneSpace
    translationMode=BMM_Additive
    translationSpace=BCS_BoneSpace
    alphaBoolBlend {blendOption=Linear}
    alphaScaleBiasClamp {bInterpResult=true, interpSpeedIncreasing=5}
    boneToModify {boneName=dyn_skirtFront_03}
    rotation {pitch=-5}
    translation {y=-0.5}
   }
   n56 = "Transform (Modify) Bone - Bone: dyn_skirtFront_04"(ComponentPose=@n55, Alpha=@n54, Translation="-0.5,0.5,-1.5", Rotation="-15,0,-30", Scale="1,1,1") {
    rotationMode=BMM_Additive
    rotationSpace=BCS_BoneSpace
    translationMode=BMM_Additive
    translationSpace=BCS_BoneSpace
    alphaBoolBlend {blendOption=Linear}
    alphaScaleBiasClamp {bInterpResult=true, interpSpeedIncreasing=5}
    boneToModify {boneName=dyn_skirtFront_04}
    rotation {pitch=-15, roll=-30}
    translation=<-0.5 0.5 -1.5>
   }
   n57 = "Transform (Modify) Bone - Bone: dyn_skirtFront_02_l"(ComponentPose=@n56, Alpha=@n54, Translation="0,0.5,1", Rotation="-3,0,0", Scale="1,1,1") {
    rotationMode=BMM_Additive
    rotationSpace=BCS_BoneSpace
    translationMode=BMM_Additive
    translationSpace=BCS_BoneSpace
    alphaBoolBlend {blendOption=Linear}
    alphaScaleBiasClamp {bInterpResult=true, interpSpeedIncreasing=5}
    boneToModify {boneName=dyn_skirtFront_02_l}
    rotation {pitch=-3}
    translation {y=0.5, z=1}
   }
   n58 = "Transform (Modify) Bone - Bone: dyn_skirtFront_03_l"(ComponentPose=@n57, Alpha=@n54, Translation="0,0.5,-1.5", Rotation="-10,0,0", Scale="1,1,1") {
    rotationMode=BMM_Additive
    rotationSpace=BCS_BoneSpace
    translationMode=BMM_Additive
    translationSpace=BCS_BoneSpace
    alphaBoolBlend {blendOption=Linear}
    alphaScaleBiasClamp {bInterpResult=true, interpSpeedIncreasing=5}
    boneToModify {boneName=dyn_skirtFront_03_l}
    rotation {pitch=-10}
    translation {y=0.5, z=-1.5}
   }
   n59 = "Transform (Modify) Bone - Bone: dyn_skirtFront_05"(ComponentPose=@n58, Alpha=@n54, Translation="0,2,0", Rotation="0,0,0", Scale="1,1,1") {
    rotationMode=BMM_Additive
    rotationSpace=BCS_BoneSpace
    translationMode=BMM_Additive
    translationSpace=BCS_BoneSpace
    alphaBoolBlend {blendOption=Linear}
    alphaScaleBiasClamp {bInterpResult=true, interpSpeedIncreasing=5}
    boneToModify {boneName=dyn_skirtFront_05}
    translation {y=2}
   }
   n60 = "Transform (Modify) Bone - Bone: dyn_skirtFront_04_l"(ComponentPose=@n59, Alpha=@n54, Translation="0,2,0", Rotation="0,0,0", Scale="1,1,1") {
    rotationMode=BMM_Additive
    rotationSpace=BCS_BoneSpace
    translationMode=BMM_Additive
    translationSpace=BCS_BoneSpace
    alphaBoolBlend {blendOption=Linear}
    alphaScaleBiasClamp {bInterpResult=true, interpSpeedIncreasing=5}
    boneToModify {boneName=dyn_skirtFront_04_l}
    translation {y=2}
   }
   n61 = "Transform (Modify) Bone - Bone: dyn_skirtSide_01_l"(ComponentPose=@n60, Alpha=@n54, Translation="0,0,0", Rotation="-10,0,0.003", Scale="1,1,1") {
    rotationMode=BMM_Additive
    rotationSpace=BCS_BoneSpace
    translationMode=BMM_Additive
    translationSpace=BCS_BoneSpace
    alphaBoolBlend {blendOption=Linear}
    alphaScaleBiasClamp {bInterpResult=true, interpSpeedIncreasing=5}
    boneToModify {boneName=dyn_skirtSide_01_l}
    rotation {pitch=-10, roll=0.003}
   }
   n62 = "Transform (Modify) Bone - Bone: dyn_skirtSide_03_l"(ComponentPose=@n61, Alpha=@n54, Translation="-3,0,0", Rotation="0,0,0", Scale="1,1,1") {
    rotationMode=BMM_Additive
    rotationSpace=BCS_BoneSpace
    translationMode=BMM_Additive
    translationSpace=BCS_BoneSpace
    alphaBoolBlend {blendOption=Linear}
    alphaScaleBiasClamp {bInterpResult=true, interpSpeedIncreasing=5}
    boneToModify {boneName=dyn_skirtSide_03_l}
    translation {x=-3}
   }
   n63 = "Transform (Modify) Bone - Bone: dyn_skirtSide_04_l"(ComponentPose=@n62, Alpha=@n54, Translation="-4,0,0", Rotation="0,0,0", Scale="1,1,1") {
    rotationMode=BMM_Additive
    rotationSpace=BCS_BoneSpace
    translationMode=BMM_Additive
    translationSpace=BCS_BoneSpace
    alphaBoolBlend {blendOption=Linear}
    alphaScaleBiasClamp {bInterpResult=true, interpSpeedIncreasing=5}
    boneToModify {boneName=dyn_skirtSide_04_l}
    translation {x=-4}
   }
   n64 = Get crouch still()
   n65 = "Transform (Modify) Bone - Bone: dyn_skirtBack_mstr_r"!(ComponentPose=@n63, Alpha:float=@n64, Translation="0,0,0", Rotation="3,0,0", Scale="1,1,1") -> (Pose) {
    rotationMode=BMM_Additive
    rotationSpace=BCS_BoneSpace
    alphaBoolBlend {blendInTime=0.1, blendOption=Linear, blendOutTime=0.1}
    alphaScaleBiasClamp {bInterpResult=true, interpSpeedIncreasing=20}
    boneToModify {boneName=dyn_skirtBack_mstr_r}
    rotation {pitch=3}
   }
   n66 = "Transform (Modify) Bone - Bone: dyn_skirtSide_mstr_r"!(ComponentPose=@n65, Alpha:float=@n64, Translation="0,0,0", Rotation="3,0,0", Scale="1,1,1") -> (Pose) {
    rotationMode=BMM_Additive
    rotationSpace=BCS_BoneSpace
    alphaBoolBlend {blendInTime=0.1, blendOption=Linear, blendOutTime=0.1}
    alphaScaleBiasClamp {bInterpResult=true, interpSpeedIncreasing=20}
    boneToModify {boneName=dyn_skirtSide_mstr_r}
    rotation {pitch=3}
   }
   n67 = Get bIsSprinting()
   n68 = "Transform (Modify) Bone - Bone: dyn_skirtFront_03"!(ComponentPose=@n66, bAlphaBoolEnabled:bool=@n67, Translation="0,0,0", Rotation="10,0,0", Scale="1,1,1") -> (Pose) {
    alphaInputType=Bool
    rotationMode=BMM_Additive
    rotationSpace=BCS_BoneSpace
    translationMode=BMM_Additive
    translationSpace=BCS_BoneSpace
    alphaBoolBlend {blendInTime=0.1, blendOption=CircularOut, blendOutTime=0.1}
    alphaScaleBiasClamp {bInterpResult=true, interpSpeedIncreasing=5}
    boneToModify {boneName=dyn_skirtFront_03}
    rotation {pitch=10}
   }
   n69 = "Transform (Modify) Bone - Bone: dyn_skirtFront_04"!(ComponentPose=@n68, bAlphaBoolEnabled:bool=@n67, Translation="0,0,0", Rotation="10,0,0", Scale="1,1,1") -> (Pose) {
    alphaInputType=Bool
    rotationMode=BMM_Additive
    rotationSpace=BCS_BoneSpace
    translationMode=BMM_Additive
    translationSpace=BCS_BoneSpace
    alphaBoolBlend {blendInTime=0.1, blendOption=CircularOut, blendOutTime=0.1}
    alphaScaleBiasClamp {bInterpResult=true, interpSpeedIncreasing=5}
    boneToModify {boneName=dyn_skirtFront_04}
    rotation {pitch=10}
   }
   n70 = "Transform (Modify) Bone - Bone: dyn_skirtFront_05"!(ComponentPose=@n69, bAlphaBoolEnabled:bool=@n67, Translation="0,0,0", Rotation="10,0,0", Scale="1,1,1") -> (Pose) {
    alphaInputType=Bool
    rotationMode=BMM_Additive
    rotationSpace=BCS_BoneSpace
    translationMode=BMM_Additive
    translationSpace=BCS_BoneSpace
    alphaBoolBlend {blendInTime=0.1, blendOption=CircularOut, blendOutTime=0.1}
    alphaScaleBiasClamp {bInterpResult=true, interpSpeedIncreasing=5}
    boneToModify {boneName=dyn_skirtFront_05}
    rotation {pitch=10}
   }
   n71 = "Transform (Modify) Bone - Bone: dyn_skirtFront_03_l"!(ComponentPose=@n70, bAlphaBoolEnabled:bool=@n67, Translation="0,0,0", Rotation="10,0,0", Scale="1,1,1") -> (Pose) {
    alphaInputType=Bool
    rotationMode=BMM_Additive
    rotationSpace=BCS_BoneSpace
    translationMode=BMM_Additive
    translationSpace=BCS_BoneSpace
    alphaBoolBlend {blendInTime=0.1, blendOption=CircularOut, blendOutTime=0.1}
    alphaScaleBiasClamp {bInterpResult=true, interpSpeedIncreasing=5}
    boneToModify {boneName=dyn_skirtFront_03_l}
    rotation {pitch=10}
   }
   n72 = "Transform (Modify) Bone - Bone: dyn_skirtFront_04_l"!(ComponentPose=@n71, bAlphaBoolEnabled:bool=@n67, Translation="0,0,0", Rotation="10,0,0", Scale="1,1,1") -> (Pose) {
    alphaInputType=Bool
    rotationMode=BMM_Additive
    rotationSpace=BCS_BoneSpace
    translationMode=BMM_Additive
    translationSpace=BCS_BoneSpace
    alphaBoolBlend {blendInTime=0.1, blendOption=CircularOut, blendOutTime=0.1}
    alphaScaleBiasClamp {bInterpResult=true, interpSpeedIncreasing=5}
    boneToModify {boneName=dyn_skirtFront_04_l}
    rotation {pitch=10}
   }
   n73 = Get crouch move() -> (crouch move:double)
   n74 = "Transform (Modify) Bone - Bone: dyn_skirtFront_04"!(ComponentPose=@n72, Alpha:float=@n73, Translation="0,0,0", Rotation="10,0,0", Scale="1,1,1") -> (Pose) {
    rotationMode=BMM_Additive
    rotationSpace=BCS_BoneSpace
    translationMode=BMM_Additive
    translationSpace=BCS_BoneSpace
    alphaBoolBlend {blendInTime=0.1, blendOption=CircularOut, blendOutTime=0.1}
    alphaScaleBiasClamp {bInterpResult=true, interpSpeedDecreasing=20, interpSpeedIncreasing=20}
    boneToModify {boneName=dyn_skirtFront_04}
    rotation {pitch=10}
   }
   n75 = "Transform (Modify) Bone - Bone: dyn_skirtFront_05"!(ComponentPose=@n74, Alpha:float=@n73, Translation="0,0,0", Rotation="10,0,0", Scale="1,1,1") -> (Pose) {
    rotationMode=BMM_Additive
    rotationSpace=BCS_BoneSpace
    translationMode=BMM_Additive
    translationSpace=BCS_BoneSpace
    alphaBoolBlend {blendInTime=0.1, blendOption=CircularOut, blendOutTime=0.1}
    alphaScaleBiasClamp {bInterpResult=true, interpSpeedDecreasing=20, interpSpeedIncreasing=20}
    boneToModify {boneName=dyn_skirtFront_05}
    rotation {pitch=10}
   }
   n76 = "Transform (Modify) Bone - Bone: dyn_skirtFront_03"!(ComponentPose=@n75, Alpha:float=@n73, Translation="0,0,1", Rotation="0,0,0", Scale="1,1,1") -> (Pose) {
    rotationMode=BMM_Additive
    rotationSpace=BCS_BoneSpace
    translationMode=BMM_Additive
    translationSpace=BCS_BoneSpace
    alphaBoolBlend {blendInTime=0.1, blendOption=CircularOut, blendOutTime=0.1}
    alphaScaleBiasClamp {bInterpResult=true, interpSpeedDecreasing=20, interpSpeedIncreasing=20}
    boneToModify {boneName=dyn_skirtFront_03}
    translation {z=1}
   }
   n77 = Get bIsKneeSliding()
   n78 = "Transform (Modify) Bone - Bone: dyn_skirtBack_01"!(ComponentPose=@n76, bAlphaBoolEnabled:bool=@n77, Translation="0,0,0", Rotation="15,0,0", Scale="1,1,1") -> (Pose) {
    alphaInputType=Bool
    rotationMode=BMM_Additive
    rotationSpace=BCS_BoneSpace
    alphaBoolBlend {blendInTime=0.1, blendOption=Linear, blendOutTime=0.1}
    boneToModify {boneName=dyn_skirtBack_01}
    rotation {pitch=15}
   }
   n79 = "Transform (Modify) Bone - Bone: dyn_skirtBack_01_l"!(ComponentPose=@n78, bAlphaBoolEnabled:bool=@n77, Translation="0,0,0", Rotation="15,0,0", Scale="1,1,1") -> (Pose) {
    alphaInputType=Bool
    rotationMode=BMM_Additive
    rotationSpace=BCS_BoneSpace
    alphaBoolBlend {blendInTime=0.1, blendOption=Linear, blendOutTime=0.1}
    boneToModify {boneName=dyn_skirtBack_01_l}
    rotation {pitch=15}
   }
   n80 = "Transform (Modify) Bone - Bone: dyn_skirtBack_01_r"!(ComponentPose=@n79, bAlphaBoolEnabled:bool=@n77, Translation="0,0,0", Rotation="15,0,0", Scale="1,1,1") -> (Pose) {
    alphaInputType=Bool
    rotationMode=BMM_Additive
    rotationSpace=BCS_BoneSpace
    alphaBoolBlend {blendInTime=0.1, blendOption=Linear, blendOutTime=0.1}
    boneToModify {boneName=dyn_skirtBack_01_r}
    rotation {pitch=15}
   }
   n81 = "Transform (Modify) Bone - Bone: dyn_scarf_MSTR"!(ComponentPose=@n80, bAlphaBoolEnabled:bool=Get bIsTacticalSprinting(), Rotation="15,-15,0") -> (Pose) {
    alphaInputType=Bool
    rotationMode=BMM_Replace
    rotationSpace=BCS_BoneSpace
    alphaBoolBlend {blendInTime=0.1, blendOption=Linear, blendOutTime=0.1}
    boneToModify {boneName=dyn_scarf_MSTR}
    rotation {pitch=15, yaw=-15}
   }
   n82 = "Transform (Modify) Bone - Bone: dyn_scarf_MSTR"(ComponentPose=@n81, bAlphaBoolEnabled=Get bIsInFrontEnd(), Rotation="5,-5,0") {
    alphaInputType=Bool
    rotationMode=BMM_Replace
    rotationSpace=BCS_BoneSpace
    alphaBoolBlend {blendInTime=0.1, blendOption=Linear, blendOutTime=0.1}
    boneToModify {boneName=dyn_scarf_MSTR}
    rotation {pitch=5, yaw=-5}
   }
   n83 = Anim Dynamics - Bone: dyn_skirtFront_04(ComponentPose=@n82, Alpha=1, ExternalForce="0,0,0") {
    angularDampingOverride=0.99
    angularSpringConstant=50
    bAngularSpring=true
    bGravityOverrideInSimSpace=true
    bOverrideAngularDamping=true
    simulationSpace=RootRelative
    alphaBoolBlend {blendOption=Linear}
    boundBone {boneName=dyn_skirtFront_04}
    componentAppliedLinearAccClamp=<3500 3500 3500>
    componentLinearAccScale=<0.4 0.4 0.8>
    componentLinearVelScale=<0.4 0.4 0.8>
    gravityOverride {z=-810}
    physicsBodyDefinitions [{
      sphereCollisionRadius=0
      boundBone {boneName=dyn_skirtFront_04}
      boxExtents=<20 25 25>
      constraintSetup {angularLimitsMax=<5 0 3>, angularLimitsMin=<-5 -35 -35>, angularTarget=<1 0 0>}
      localJointOffset=<3 0 0>
     }]
    relativeSpaceBone {boneName=root}
   }
   n84 = Anim Dynamics - Bone: dyn_skirtFront_03_l(ComponentPose=@n83, Alpha=10, ExternalForce="0,0,0") {
    angularDampingOverride=0.99
    angularSpringConstant=50
    bAngularSpring=true
    bGravityOverrideInSimSpace=true
    bOverrideAngularDamping=true
    simulationSpace=RootRelative
    alphaBoolBlend {blendOption=Linear}
    boundBone {boneName=dyn_skirtFront_03_l}
    componentAppliedLinearAccClamp=<3500 3500 3500>
    componentLinearAccScale=<0.4 0.4 0.8>
    componentLinearVelScale=<0.4 0.4 0.8>
    gravityOverride {z=-810}
    physicsBodyDefinitions [{
      sphereCollisionRadius=0
      boundBone {boneName=dyn_skirtFront_03_l}
      boxExtents=<20 25 25>
      constraintSetup {angularLimitsMax=<5 0 5>, angularLimitsMin=<-5 -35 -25>, angularTarget=<1 0 0>}
      localJointOffset=<3 0 0>
     }]
    relativeSpaceBone {boneName=root}
   }
   n85 = Anim Dynamics - Bone: dyn_skirtFront_05(ComponentPose=@n84, Alpha=1, ExternalForce="0,0,0") {
    angularDampingOverride=0.99
    angularSpringConstant=50
    bAngularSpring=true
    bGravityOverrideInSimSpace=true
    bOverrideAngularDamping=true
    simulationSpace=RootRelative
    alphaBoolBlend {blendOption=Linear}
    boundBone {boneName=dyn_skirtFront_05}
    componentAppliedLinearAccClamp=<3500 3500 3500>
    componentLinearAccScale=<0.4 0.4 0.8>
    componentLinearVelScale=<0.4 0.4 0.8>
    gravityOverride {z=-810}
    physicsBodyDefinitions [{
      sphereCollisionRadius=0
      boundBone {boneName=dyn_skirtFront_05}
      boxExtents=<20 25 25>
      constraintSetup {angularLimitsMax=<5 3 3>, angularLimitsMin=<-5 -35 -35>, angularTarget=<1 0 0>}
      localJointOffset=<3 0 0>
     }]
    relativeSpaceBone {boneName=root}
   }
   n86 = Anim Dynamics - Bone: dyn_skirtFront_04_l(ComponentPose=@n85, Alpha=10, ExternalForce="0,0,0") {
    angularDampingOverride=0.99
    angularSpringConstant=50
    bAngularSpring=true
    bGravityOverrideInSimSpace=true
    bOverrideAngularDamping=true
    simulationSpace=RootRelative
    alphaBoolBlend {blendOption=Linear}
    boundBone {boneName=dyn_skirtFront_04_l}
    componentAppliedLinearAccClamp=<3500 3500 3500>
    componentLinearAccScale=<0.4 0.4 0.8>
    componentLinearVelScale=<0.4 0.4 0.8>
    gravityOverride {z=-810}
    physicsBodyDefinitions [{
      sphereCollisionRadius=0
      boundBone {boneName=dyn_skirtFront_04_l}
      boxExtents=<20 25 25>
      constraintSetup {angularLimitsMax=<5 3 5>, angularLimitsMin=<-5 -35 -25>, angularTarget=<1 0 0>}
      localJointOffset=<3 0 0>
     }]
    relativeSpaceBone {boneName=root}
   }
   n87 = Anim Dynamics - Bone: dyn_quiver(ComponentPose=@n86, Alpha=1, ExternalForce="0,0,0") {
    angularDampingOverride=0.98
    angularSpringConstant=50
    bAngularSpring=true
    bGravityOverrideInSimSpace=true
    bOverrideAngularDamping=true
    bUseSphericalLimits=true
    simulationSpace=RootRelative
    alphaBoolBlend {blendOption=Linear}
    boundBone {boneName=dyn_quiver}
    componentAppliedLinearAccClamp=<3500 3500 3500>
    componentLinearAccScale=<0.8 0.8 0.8>
    componentLinearVelScale=<0.8 0.8 0.8>
    physicsBodyDefinitions [{
      collisionType=CustomSphere
      sphereCollisionRadius=4
      boundBone {boneName=dyn_quiver}
      boxExtents=<25 20 20>
      constraintSetup {angularLimitsMax=<1 5 10>, angularLimitsMin=<-1 -25 -10>, angularTarget=<1 0 0>}
      localJointOffset=<1 0 0>
     }]
    relativeSpaceBone {boneName=pelvis}
    sphericalLimits [{limitRadius=6, drivingBone {boneName=upperarm_twist_02_r}, sphereLocalOffset=<0 0 3>}]
   }
   n88 = Anim Dynamics - Bone: dyn_moon_charm(ComponentPose=@n87, Alpha=1, ExternalForce="0,0,0") {
    angularDampingOverride=0.98
    angularSpringConstant=75
    bAngularSpring=true
    bGravityOverrideInSimSpace=true
    bOverrideAngularDamping=true
    simulationSpace=RootRelative
    alphaBoolBlend {blendOption=Linear}
    boundBone {boneName=dyn_moon_charm}
    componentAppliedLinearAccClamp=<3500 3500 3500>
    componentLinearAccScale=<0.9 0.9 0.9>
    componentLinearVelScale=<0.9 0.9 0.9>
    physicsBodyDefinitions [{
      sphereCollisionRadius=0
      boundBone {boneName=dyn_moon_charm}
      boxExtents=<25 15 15>
      constraintSetup {angularLimitsMax=<1 5 25>, angularLimitsMin=<-1 -35 -25>, angularTarget=<1 0 0>}
      localJointOffset=<2 0 0>
     }]
    relativeSpaceBone {boneName=pelvis}
   }
   n89 = Anim Dynamics - Bone: dyn_moon_charm01(ComponentPose=@n88, Alpha=1, ExternalForce="0,0,0") {
    angularDampingOverride=0.98
    angularSpringConstant=75
    bAngularSpring=true
    bGravityOverrideInSimSpace=true
    bOverrideAngularDamping=true
    bUseSphericalLimits=true
    simulationSpace=RootRelative
    alphaBoolBlend {blendOption=Linear}
    boundBone {boneName=dyn_moon_charm01}
    componentAppliedLinearAccClamp=<3500 3500 3500>
    componentLinearAccScale=<0.9 0.9 0.9>
    componentLinearVelScale=<0.9 0.9 0.9>
    physicsBodyDefinitions [{
      collisionType=CustomSphere
      sphereCollisionRadius=4
      boundBone {boneName=dyn_moon_charm01}
      boxExtents=<25 15 15>
      constraintSetup {angularLimitsMax=<1 5 25>, angularLimitsMin=<-1 -35 -25>, angularTarget=<1 0 0>}
      localJointOffset=<2 0 0>
     }]
    relativeSpaceBone {boneName=pelvis}
    sphericalLimits [{limitRadius=6, drivingBone {boneName=upperarm_twist_02_r}, sphereLocalOffset=<0 0 3>}]
   }
   n90 = Anim Dynamics - Bone: dyn_arrow_a(ComponentPose=@n89, Alpha=0.5, ExternalForce="0,0,0") {
    angularDampingOverride=0.99
    angularSpringConstant=50
    bAngularSpring=true
    bGravityOverrideInSimSpace=true
    bOverrideAngularDamping=true
    bUseSphericalLimits=true
    simulationSpace=RootRelative
    alphaBoolBlend {blendOption=Linear}
    boundBone {boneName=dyn_arrow_a}
    componentAppliedLinearAccClamp=<3500 3500 3500>
    componentLinearAccScale=<0.5 0.5 0.5>
    componentLinearVelScale=<0.5 0.5 0.5>
    physicsBodyDefinitions [{
      collisionType=CustomSphere
      sphereCollisionRadius=4
      boundBone {boneName=dyn_arrow_a}
      boxExtents=<15 15 15>
      constraintSetup {angularLimitsMax=<2 1 1>, angularLimitsMin=<-2 -1 -3>, angularTarget=<1 0 0>}
      localJointOffset=<0.5 0 0>
     }]
    relativeSpaceBone {boneName=pelvis}
    sphericalLimits [{limitRadius=6, drivingBone {boneName=upperarm_twist_02_r}, sphereLocalOffset=<0 0 3>}]
   }
   n91 = Anim Dynamics - Bone: dyn_arrow_b(ComponentPose=@n90, Alpha=0.5, ExternalForce="0,0,0") {
    angularDampingOverride=0.99
    angularSpringConstant=50
    bAngularSpring=true
    bGravityOverrideInSimSpace=true
    bOverrideAngularDamping=true
    bUseSphericalLimits=true
    simulationSpace=RootRelative
    alphaBoolBlend {blendOption=Linear}
    boundBone {boneName=dyn_arrow_b}
    componentAppliedLinearAccClamp=<3500 3500 3500>
    componentLinearAccScale=<0.5 0.5 0.5>
    componentLinearVelScale=<0.5 0.5 0.5>
    physicsBodyDefinitions [{
      collisionType=CustomSphere
      sphereCollisionRadius=4
      boundBone {boneName=dyn_arrow_b}
      boxExtents=<15 15 15>
      constraintSetup {angularLimitsMax=<-2 1 1>, angularLimitsMin=<-3 -1 -2>, angularTarget=<1 0 0>}
      localJointOffset=<0.25 0 0>
     }]
    relativeSpaceBone {boneName=pelvis}
    sphericalLimits [{limitRadius=6, drivingBone {boneName=upperarm_twist_02_r}, sphereLocalOffset=<0 0 3>}]
   }
   n92 = Anim Dynamics - Bone: dyn_arrow_c(ComponentPose=@n91, Alpha=0.5, ExternalForce="0,0,0") {
    angularDampingOverride=0.99
    angularSpringConstant=50
    bAngularSpring=true
    bGravityOverrideInSimSpace=true
    bOverrideAngularDamping=true
    bUseSphericalLimits=true
    simulationSpace=RootRelative
    alphaBoolBlend {blendOption=Linear}
    boundBone {boneName=dyn_arrow_c}
    componentAppliedLinearAccClamp=<3500 3500 3500>
    componentLinearAccScale=<0.5 0.5 0.5>
    componentLinearVelScale=<0.5 0.5 0.5>
    physicsBodyDefinitions [{
      collisionType=CustomSphere
      sphereCollisionRadius=4
      boundBone {boneName=dyn_arrow_c}
      boxExtents=<15 15 15>
      constraintSetup {angularLimitsMax=<1 1 1>, angularLimitsMin=<-1 -1 -1>, angularTarget=<1 0 0>}
      localJointOffset=<0.75 0 0>
     }]
    relativeSpaceBone {boneName=pelvis}
    sphericalLimits [{limitRadius=6, drivingBone {boneName=upperarm_twist_02_r}, sphereLocalOffset=<0 0 3>}]
   }
   n93 = "Transform (Modify) Bone - Bone: dyn_moon_charm01"(ComponentPose=@n92, Alpha=Get crouch still(), Translation="0,0,0", Rotation="30,0,0", Scale="1,1,1") {
    rotationMode=BMM_Additive
    rotationSpace=BCS_BoneSpace
    alphaBoolBlend {blendInTime=0.1, blendOption=Linear, blendOutTime=0.1}
    alphaScaleBiasClamp {bInterpResult=true, interpSpeedIncreasing=30}
    boneToModify {boneName=dyn_moon_charm01}
    rotation {pitch=30}
   }
   AnimGraphNode_SaveCachedPose_customIdle_0 = Save cached pose 'startpose'(Pose=Component To Local(ComponentPose=@n93)) {cacheName=startpose}
  ]}
 EventGraph {nodes [
   Event BlueprintUpdateAnimation()
   > n1 = Sequence!(execute) -> (then_0, then_1, then_2:Exec)
   > n2 = Gravity Override State Function DEPRECATED(
    CharacterPartAnimInstance=Self-Reference()
    GameTimeInSeconds=GetGameTimeInSeconds()
    GravityOverrideParams=Get GravityOverrideParams scarf() -> (GravityOverrideParams scarf:Gravity Override Params Struct Structure)
   )
   > Set GravityOverride scarf(GravityOverride scarf:Vector=@n2) -> (Output_Get:Vector)
   > n3 = Gravity Override State Function DEPRECATED(CharacterPartAnimInstance=Self-Reference(), GameTimeInSeconds=GetGameTimeInSeconds(), GravityOverrideParams=Get Gravity Override Params())
   > Set GravityOverride(GravityOverride=@n3)
   n4 = NOT Boolean(A=Get bIsParachuteOpened())
   n5 = OR Boolean(A=Get bIsJumping(), B=Get bIsFalling())
   > Set jumping(jumping:double=SelectFloat(A=1, B=0, bPickA=AND Boolean(A=@n5, B=@n4))) -> (Output_Get:double)
   > Set jumping_chain(jumping_chain:double=SelectFloat(A=1, B=0, bPickA=@n5)) -> (Output_Get:double)
   n6 = NOT Boolean(A=Get bIsMoving2D())
   n7 = Get bIsCrouching()
   > Set crouch still(crouch still:double=SelectFloat(A=1, B=0, bPickA=AND Boolean(A=@n7, B=@n6))) -> (Output_Get:double)
   n8 = Get bIsCrouching()
   n9 = Get bIsMoving2D()
   > Set crouch move(crouch move:double=SelectFloat(A=1, B=0, bPickA=AND Boolean(A=@n8, B=@n9))) -> (Output_Get:double)
   > Set melee anim(melee anim:double=SelectFloat(A=0.5, B=0, bPickA=Get bIsPlayingMeleeAnim())) -> (Output_Get:double)
   n10 = NOR Boolean!(A:bool=Get bIsPlayingEmote(), B:bool=Get bIsSkyDiving()) -> (ReturnValue:bool)
   n11 = Get bIsInFrontEnd()
   > Set FrontendIdle(FrontendIdle:float=SelectFloat(A=0, B=1, bPickA=AND Boolean(A=@n11, B=@n10))) -> (Output_Get:float)
   > n12 = Branch(Condition=CheckForBackbling())
   > Set ACC(ACC:Vector="0, 0, 0") -> (Output_Get:Vector)
   > Set VEL(VEL:Vector="0, 0, 0") -> (Output_Get:Vector)
   > Set backpack(backpack:double=0.5) -> (Output_Get:double)
   > Set backpack_bool(backpack_bool:bool=true) -> (Output_Get:bool)
   Set ACC(execute=@n12.else, ACC="0.2,0.2,0.5")
   > Set VEL(VEL="0.2,0.2,0.4")
   > Set backpack(backpack=0)
   > Set backpack_bool(backpack_bool=false)
   n13 = ComplexSimSpaceFunction DEPRECATED!(
    execute=@n1.then_1
    CharacterPartAnimInstance:Custom Character Part Anim Instance Object Reference=Self-Reference()
    InterpSpeedMax:double=9
    PawnSpeedMax:double=20
    RInterpErrorTolerance:double=1
    DeltaTime:double=0
    DebugDraw:bool=false
    sim_space_rotation_in:Rotator=Get SimSpaceRotation() -> (SimSpaceRotation:Rotator)
    inComponentSpaceIn:bool=Get bInComponentSpace() -> (bInComponentSpace:bool)
    inComponentSpaceLastFrameIn:bool=Get bInComponentSpaceLastFrame() -> (bInComponentSpaceLastFrame:bool)
    lockedOnSimSpaceIn:bool=Get bLockedOnSimSpace() -> (bLockedOnSimSpace:bool)
    InterpSpeedIn:double=Get InterpSpeed() -> (InterpSpeed:float)
   ) -> (
    then
    sim_space_rotation:Rotator
    inComponentSpaceOut:bool
    inComponentSpaceLastFrameOut:bool
    lockedOnSimSpaceOut:bool
    InterpSpeedOut:double
   )
   > Set SimSpaceRotation(SimSpaceRotation:Rotator=@n13) -> (Output_Get:Rotator)
   > Set bInComponentSpace(bInComponentSpace:bool=@n13.inComponentSpaceOut) -> (Output_Get:bool)
   > Set bInComponentSpaceLastFrame(bInComponentSpaceLastFrame:bool=@n13.inComponentSpaceLastFrameOut) -> (Output_Get:bool)
   > Set bLockedOnSimSpace(bLockedOnSimSpace:bool=@n13.lockedOnSimSpaceOut) -> (Output_Get:bool)
   > Set InterpSpeed(InterpSpeed:float=@n13.InterpSpeedOut) -> (Output_Get:float)
   K2Node_VariableGet_customIdle_0 = Get bIsInFrontEnd()
   K2Node_IfThenElse_customIdle_0 = Branch(execute=@n1.then_2, Condition=@K2Node_VariableGet_customIdle_0)
   K2Node_CallFunction_customIdle_0 = GetOwnerFrontEndAnimInstance()
   K2Node_CallFunction_customIdle_1 = GetCurveValue(self=@K2Node_CallFunction_customIdle_0, CurveName=Frontend_Idle_Bypass)
   > K2Node_VariableSet_customIdle_0 = Set Custom Idle is Playing(Custom Idle is Playing:float=@K2Node_CallFunction_customIdle_1) -> (Output_Get:float)
   Event OnCharacterCustomizationLoaded()
   > Delay(Duration=0.1)
   > ResetDynamics(InTeleportType=TeleportPhysics)
   > ResetDynamics(InTeleportType=ResetPhysics)
   AND Boolean(A=OR Boolean(A=Get bIsJumping(), B=Get bIsFalling()), B=NOT Boolean(A=Get bIsParachuteOpened()))
  ]}
]
variables [
 ACC {type=Vector}
 backpack {type=number}
 backpack_bool {type=boolean}
 bInComponentSpace {type=boolean}
 bInComponentSpaceLastFrame {type=boolean}
 bLockedOnSimSpace {type=boolean}
 crouch move {type=number}
 crouch still {type=number}
 Custom Idle is Playing {type=number}
 FrontendIdle {type=number}
 Gravity Override Params {
  type=GravityOverrideParamsStruct
  default {
   skydive_FloatingDown {gravityOverride=<0 -100 -270>, windAmplitude=<75 -150 100>, windFrequency=<0.75 3.5 4>, windOffset=<0 -300 200>}
   skydive_GoingUp {gravityOverride=<0 -50 100>, windAmplitude=<75 -150 100>, windFrequency=<0.75 3.5 4>}
   skydive_Parachuting {gravityOverride=<0 -400 -980>, windAmplitude=<75 -150 100>, windFrequency=<0.75 3.5 4>}
  }
 }
 GravityOverride {type=Vector}
 GravityOverride scarf {type=Vector}
 GravityOverrideParams scarf {
  type=GravityOverrideParamsStruct
  default {onGround_Moving {gravityOverride=<0 -540 -720>, windAmplitude=<0 -150 -150>, windFrequency=<0 4 4>, windOffset=<0 -900 -900>}}
 }
 InterpSpeed {type=number}
 jumping {type=number}
 jumping_chain {type=number}
 melee anim {type=number}
 SimSpaceRotation {type=Rotator}
 VEL {type=Vector}
]
